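# Level Sequence for authoring the SquibblySquid skin's main-menu animation on a spawnable FortPlayerMannequin.
# A skeletal animation section spanning frames 0–380 supplies the base motion, with a layered Fortnite biped Control Rig track over it.
# A face Control Rig track keys jaw, gaze, blinks, eyelids, brows and an asymmetric mouth-corner expression.
# Most animated face channels return to their starting values by the final frame.
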LevelSequence /BRCosmetics/Animation/Game/MainPlayer/Menu/BR/CustomSkins/_s40/SquibblySquidM/SquibblySquid_Anim/SquibblySquid_LevelSequence
bindings [FortPlayerMannequin {
  id=#1
  possessed_class=FortPlayerMannequin
  tracks [
   {type=MovieSceneSpawnTrack, sections [{type=MovieSceneSpawnSection, channels [None {type=MovieSceneScriptingBoolChannel, default=true}]}]}
   {
    type=MovieScene3DTransformTrack
    sections [{
      type=MovieScene3DTransformSection
      channels [
       Location.X {type=MovieSceneScriptingDoubleChannel, default=0, const [0,2]}
       Location.Y {type=MovieSceneScriptingDoubleChannel, default=0, const [0,2]}
       Location.Z {type=MovieSceneScriptingDoubleChannel, default=0, const [0,2]}
       Rotation.X {type=MovieSceneScriptingDoubleChannel, default=0, const [0,2]}
       Rotation.Y {type=MovieSceneScriptingDoubleChannel, default=0, const [0,2]}
       Rotation.Z {type=MovieSceneScriptingDoubleChannel, default=0, const [0,2]}
       Scale.X {type=MovieSceneScriptingDoubleChannel, default=1, const [1,3]}
       Scale.Y {type=MovieSceneScriptingDoubleChannel, default=1, const [1,3]}
       Scale.Z {type=MovieSceneScriptingDoubleChannel, default=1, const [1,3]}
      ]
     }]
   }
   {type=MovieSceneSkeletalAnimationTrack, sections [{type=MovieSceneSkeletalAnimationSection, end_frame=380, start_frame=0}]}
   {
    type=MovieSceneControlRigParameterTrack
    display_name=Fortnite_3L_Face_CtrlRig
    sections [{
      type=MovieSceneControlRigParameterSection
      channels [
       CTRL_C_eyesAim.Location.X {type=MovieSceneScriptingFloatChannel, default=0}
       CTRL_C_eyesAim.Location.Y {type=MovieSceneScriptingFloatChannel, default=0}
       CTRL_C_eyesAim.Location.Z {type=MovieSceneScriptingFloatChannel, default=0}
       CTRL_C_eyesAim.Rotation.X {type=MovieSceneScriptingFloatChannel, default=0}
       CTRL_C_eyesAim.Rotation.Y {type=MovieSceneScriptingFloatChannel, default=0}
       CTRL_C_eyesAim.Rotation.Z {type=MovieSceneScriptingFloatChannel, default=0}
       CTRL_C_eyesAim.Scale.X {type=MovieSceneScriptingFloatChannel, default=1}
       CTRL_C_eyesAim.Scale.Y {type=MovieSceneScriptingFloatChannel, default=1}
       CTRL_C_eyesAim.Scale.Z {type=MovieSceneScriptingFloatChannel, default=1}
       CTRL_L_eyeAim.Location.X {type=MovieSceneScriptingFloatChannel, default=0}
       CTRL_L_eyeAim.Location.Y {type=MovieSceneScriptingFloatChannel, default=0}
       CTRL_L_eyeAim.Location.Z {type=MovieSceneScriptingFloatChannel, default=0}
       CTRL_L_eyeAim.Rotation.X {type=MovieSceneScriptingFloatChannel, default=0}
       CTRL_L_eyeAim.Rotation.Y {type=MovieSceneScriptingFloatChannel, default=0}
       CTRL_L_eyeAim.Rotation.Z {type=MovieSceneScriptingFloatChannel, default=0}
       CTRL_L_eyeAim.Scale.X {type=MovieSceneScriptingFloatChannel, default=1}
       CTRL_L_eyeAim.Scale.Y {type=MovieSceneScriptingFloatChannel, default=1}
       CTRL_L_eyeAim.Scale.Z {type=MovieSceneScriptingFloatChannel, default=1}
       CTRL_R_eyeAim.Location.X {type=MovieSceneScriptingFloatChannel, default=0}
       CTRL_R_eyeAim.Location.Y {type=MovieSceneScriptingFloatChannel, default=0}
       CTRL_R_eyeAim.Location.Z {type=MovieSceneScriptingFloatChannel, default=0}
       CTRL_R_eyeAim.Rotation.X {type=MovieSceneScriptingFloatChannel, default=0}
       CTRL_R_eyeAim.Rotation.Y {type=MovieSceneScriptingFloatChannel, default=0}
       CTRL_R_eyeAim.Rotation.Z {type=MovieSceneScriptingFloatChannel, default=0}
       CTRL_R_eyeAim.Scale.X {type=MovieSceneScriptingFloatChannel, default=1}
       CTRL_R_eyeAim.Scale.Y {type=MovieSceneScriptingFloatChannel, default=1}
       CTRL_R_eyeAim.Scale.Z {type=MovieSceneScriptingFloatChannel, default=1}
       CTRL_convergenceSwitch {type=MovieSceneScriptingFloatChannel, default=0}
       CTRL_Master.Location.X {type=MovieSceneScriptingFloatChannel, default=0}
       CTRL_Master.Location.Y {type=MovieSceneScriptingFloatChannel, default=0}
       CTRL_Master.Location.Z {type=MovieSceneScriptingFloatChannel, default=0}
       CTRL_Master.Rotation.X {type=MovieSceneScriptingFloatChannel, default=0}
       CTRL_Master.Rotation.Y {type=MovieSceneScriptingFloatChannel, default=0}
       CTRL_Master.Rotation.Z {type=MovieSceneScriptingFloatChannel, default=0}
       CTRL_Master.Scale.X {type=MovieSceneScriptingFloatChannel, default=1}
       CTRL_Master.Scale.Y {type=MovieSceneScriptingFloatChannel, default=1}
       CTRL_Master.Scale.Z {type=MovieSceneScriptingFloatChannel, default=1}
       CTRL_C_jaw.X {type=MovieSceneScriptingFloatChannel, default=0, keys [[0,-0.9022],[98,-0.9022],[104,-0.0008],[140,-0.9513],[375,-0.9022],[380,-0.9022]]}
       CTRL_C_jaw.Y {type=MovieSceneScriptingFloatChannel, default=0, thinned_from=6, keys [[0,0],[104,0],[140,0.0191],[375,0],[380,0]]}
       CTRL_C_jaw_openExtreme {type=MovieSceneScriptingFloatChannel, default=0}
       CTRL_C_jaw_fwdBack {type=MovieSceneScriptingFloatChannel, default=0}
       CTRL_R_jaw_ChinRaiseD {type=MovieSceneScriptingFloatChannel, default=0}
       CTRL_R_jaw_ChinRaiseU {type=MovieSceneScriptingFloatChannel, default=0}
       CTRL_L_jaw_ChinRaiseU {type=MovieSceneScriptingFloatChannel, default=0}
       CTRL_L_jaw_ChinRaiseD {type=MovieSceneScriptingFloatChannel, default=0}
       CTRL_R_mouth_lipsBlow {type=MovieSceneScriptingFloatChannel, default=0}
       CTRL_L_mouth_lipsTogetherU {type=MovieSceneScriptingFloatChannel, default=0}
       CTRL_L_mouth_lipsTogetherD {type=MovieSceneScriptingFloatChannel, default=0}
       CTRL_R_mouth_lipsTogetherU {type=MovieSceneScriptingFloatChannel, default=0}
       CTRL_L_mouth_lipsBlow {type=MovieSceneScriptingFloatChannel, default=0}
       CTRL_R_mouth_lipsTogetherD {type=MovieSceneScriptingFloatChannel, default=0}
       CTRL_C_tongue_roll.X {type=MovieSceneScriptingFloatChannel, default=0, const [0,3]}
       CTRL_C_tongue_roll.Y {type=MovieSceneScriptingFloatChannel, default=0, const [0,3]}
       CTRL_C_tongue.X {type=MovieSceneScriptingFloatChannel, default=0, const [0,3]}
       CTRL_C_tongue.Y {type=MovieSceneScriptingFloatChannel, default=0, const [0,3]}
       CTRL_C_tongue_narrowWide {type=MovieSceneScriptingFloatChannel, default=0, const [0,3]}
       CTRL_C_tongue_inOut {type=MovieSceneScriptingFloatChannel, default=0}
       CTRL_R_nose.X {type=MovieSceneScriptingFloatChannel, default=0, const [0,6]}
       CTRL_R_nose.Y {type=MovieSceneScriptingFloatChannel, default=0, const [0,6]}
       CTRL_L_nose.X {type=MovieSceneScriptingFloatChannel, default=0, const [0,6]}
       CTRL_L_nose.Y {type=MovieSceneScriptingFloatChannel, default=0, const [0,6]}
       CTRL_L_eye.X {type=MovieSceneScriptingFloatChannel, default=0, const [0,6]}
       CTRL_L_eye.Y {type=MovieSceneScriptingFloatChannel, default=0, const [0,6]}
       CTRL_L_eye_blink {
        type=MovieSceneScriptingFloatChannel
        default=0.4257
        thinned_from=24
        keys [[0,0.4257],[94,0.4257],[97,0.8126],[100,0.8126],[102,0.3887],[105,0.2403],[108,0.2258],[120,0.2258],[123,0.8194],[127,0.5762],[130,0.3177],[155,0.3177],[161,0.7919],[163,0.7919],[169,0.4133],[180,1],[266,1],[275,0.54],[278,0.4133],[380,0.4257]]
       }
       CTRL_R_eye.X {type=MovieSceneScriptingFloatChannel, default=0, const [0,6]}
       CTRL_R_eye.Y {type=MovieSceneScriptingFloatChannel, default=0, const [0,6]}
       CTRL_R_eye_blink {
        type=MovieSceneScriptingFloatChannel
        default=0
        keys [[0,0],[94,0],[97,0.8041],[100,0.8041],[102,0.4033],[105,0.263],[108,0.2485],[120,0.2485],[123,0.8312],[127,0.5835],[130,0.406],[135,0.2924],[140,0.2431],[152,0.2635],[161,0.8651],[164,0.8651],[169,0.4061],[180,0.4088],[266,0.4088],[275,0.4073],[278,0.4061],[380,0]]
       }
       CTRL_C_eye.X {
        type=MovieSceneScriptingFloatChannel
        default=0
        thinned_from=52
        keys [[0,0.4236],[28,0.4236],[30,0.2028],[52,0.2028],[53,0.2825],[54,0.3198],[79,0.3198],[80,0.2771],[81,0.2432],[95,0.2497],[105,0.2497],[107,0.0621],[123,0.0621],[129,0.2855],[132,0.2932],[134,0.3187],[137,0.3217],[146,0.3217],[147,0.2573],[148,0.2316],[155,0.2316],[157,-0.0046],[165,0.1737],[177,-0.4071],[186,-0.8172],[229,-0.8172],[230,-0.6699],[231,-0.6225],[241,-0.8053],[242,-0.9402],[243,-1],[253,-1],[277,-0.59],[285,0.2407],[307,0.2407],[310,-0.4099],[328,-0.4099],[331,-0.1709],[350,-0.1731],[351,-0.1111],[352,0.0357],[353,0.2028],[354,0.3455],[355,0.4236],[380,0.4236]]
       }
       CTRL_C_eye.Y {
        type=MovieSceneScriptingFloatChannel
        default=0
        thinned_from=53
        keys [[0,0.1724],[28,0.1724],[30,0.2191],[52,0.2191],[54,0.1977],[79,0.1977],[80,0.2282],[81,0.2384],[93,0.1977],[95,-0.3222],[105,-0.3222],[107,-0.2917],[123,-0.2917],[129,0.0103],[132,-0.087],[134,-0.0609],[137,-0.1839],[146,-0.1839],[147,-0.1732],[148,-0.1689],[155,-0.1689],[157,-0.1443],[165,-0.3155],[177,-0.4801],[186,-0.8088],[206,-0.8088],[209,-0.7498],[211,-0.7441],[222,-0.7441],[229,-0.7873],[230,-0.7529],[231,-0.7054],[241,-0.7054],[243,-0.7073],[253,-0.7073],[277,-0.4801],[285,-0.0412],[307,-0.0412],[310,0.001],[328,0.001],[332,0.0097],[350,0.0097],[351,0.0321],[353,0.15],[354,0.1724],[380,0.1724]]
       }
       CTRL_C_eyesConvergence {type=MovieSceneScriptingFloatChannel, default=0, const [0,6]}
       CTRL_L_eye_cheekRaise {type=MovieSceneScriptingFloatChannel, default=0, const [0,7]}
       CTRL_L_eye_squintInner {type=MovieSceneScriptingFloatChannel, default=0, const [0,6]}
       CTRL_L_nose_wrinkleUpper {type=MovieSceneScriptingFloatChannel, default=0, const [0,6]}
       CTRL_R_nose_wrinkleUpper {type=MovieSceneScriptingFloatChannel, default=0, const [0,6]}
       CTRL_R_eye_squintInner {type=MovieSceneScriptingFloatChannel, default=0, const [0,6]}
       CTRL_R_eye_cheekRaise {
        type=MovieSceneScriptingFloatChannel
        default=1
        thinned_from=9
        keys [[0,0.9999],[129,0.126],[132,0.0012],[137,0],[142,0],[275,0.6055],[375,1],[380,0.9999]]
       }
       CTRL_C_eyelidsReduce {type=MovieSceneScriptingFloatChannel, default=0, const [0,6]}
       CTRL_R_brow_lateral {type=MovieSceneScriptingFloatChannel, default=0, const [0,6]}
       CTRL_L_brow_lateral {type=MovieSceneScriptingFloatChannel, default=0, const [0,6]}
       CTRL_R_brow_down {type=MovieSceneScriptingFloatChannel, default=0, keys [[0,0.3551],[83,0.3107],[129,0],[136,0.3419],[144,0.3551],[380,0.3551]]}
       CTRL_L_brow_down {type=MovieSceneScriptingFloatChannel, default=0, keys [[0,0.3471],[83,0.3037],[129,0],[136,0.334],[144,0.3471],[380,0.3471]]}
       CTRL_L_brow_raiseOut {type=MovieSceneScriptingFloatChannel, default=0, const [0,6]}
       CTRL_L_brow_raiseIn {type=MovieSceneScriptingFloatChannel, default=0, const [0,6]}
       CTRL_R_brow_raiseIn {type=MovieSceneScriptingFloatChannel, default=0, const [0,6]}
       CTRL_R_brow_raiseOut {type=MovieSceneScriptingFloatChannel, default=0, const [0,6]}
       CTRL_lookAtSwitch {type=MovieSceneScriptingFloatChannel, default=0}
       CTRL_C_teethD.X {type=MovieSceneScriptingFloatChannel, default=0, const [0,3]}
       CTRL_C_teethD.Y {type=MovieSceneScriptingFloatChannel, default=0, const [0,3]}
       CTRL_C_teethU.X {type=MovieSceneScriptingFloatChannel, default=0, const [0,3]}
       CTRL_C_teethU.Y {type=MovieSceneScriptingFloatChannel, default=0, const [0,3]}
       CTRL_C_teeth_fwdBackU {type=MovieSceneScriptingFloatChannel, default=0, const [0,3]}
       CTRL_L_mouth_corner.X {type=MovieSceneScriptingFloatChannel, default=0, keys [[0,0],[129,0],[142,0.7315],[258,0.7315],[380,0]]}
       CTRL_L_mouth_corner.Y {type=MovieSceneScriptingFloatChannel, default=0, keys [[0,0],[129,0],[142,0.2629],[258,0.2629],[380,0]]}
       CTRL_R_mouth_corner.X {type=MovieSceneScriptingFloatChannel, default=0, keys [[0,0],[129,0],[142,0.9886],[258,0.9886],[380,0]]}
       CTRL_R_mouth_corner.Y {type=MovieSceneScriptingFloatChannel, default=0, keys [[0,0],[129,0],[142,1],[258,1],[380,0]]}
       CTRL_C_teeth_fwdBackD {type=MovieSceneScriptingFloatChannel, default=0}
       CTRL_R_mouth_cornerSharpnessD {type=MovieSceneScriptingFloatChannel, default=0, const [0,3]}
       CTRL_R_mouth_cornerSharpnessU {type=MovieSceneScriptingFloatChannel, default=0, const [0,3]}
       CTRL_L_eye_eyelidU.X {
        type=MovieSceneScriptingFloatChannel
        default=-0.6465
        thinned_from=25
        keys [[0,-0.6465],[92,-0.6465],[99,-0.5718],[114,-0.6198],[124,-0.5793],[128,-0.6464],[163,-0.6465],[165,-0.6542],[167,-0.6197],[175,-0.432],[186,-0.6148],[266,-0.6148],[275,-0.4015],[278,-0.3949],[285,-0.5173],[305,-0.5233],[309,-0.5039],[318,-0.5302],[350,-0.5349],[369,-0.6325],[380,-0.6465]]
       }
       CTRL_L_eye_eyelidU.Y {
        type=MovieSceneScriptingFloatChannel
        default=-0.0356
        thinned_from=25
        keys [[0,-0.0356],[92,-0.0356],[99,-0.4081],[114,-0.1653],[124,-0.4267],[128,-0.0362],[133,-0.0356],[163,-0.0356],[165,-0.0857],[167,0.0652],[175,0.7712],[186,0.333],[266,0.333],[275,-0.547],[278,-0.8424],[285,0.3262],[288,0.6827],[305,0.6699],[309,0.1817],[318,0.5161],[350,0.5854],[369,0.0421],[380,-0.0356]]
       }
       CTRL_L_mouth_cornerSharpnessD {type=MovieSceneScriptingFloatChannel, default=0, const [0,3]}
       CTRL_L_eye_eyelidD.X {
        type=MovieSceneScriptingFloatChannel
        default=0
        thinned_from=19
        keys [[0,0],[142,0],[161,0.3949],[165,0.4852],[169,0.5564],[175,0.6007],[186,0.437],[266,0.437],[272,0.4037],[275,0.378],[276,0.3591],[296,0.1492],[350,0],[380,0]]
       }
       CTRL_L_eye_eyelidD.Y {
        type=MovieSceneScriptingFloatChannel
        default=0
        thinned_from=19
        keys [[0,0],[142,0],[161,0.029],[165,0.0933],[169,0.6109],[175,0.6595],[186,-1],[266,-1],[272,-0.5602],[275,-0.1566],[276,-0.3619],[296,0.1638],[350,0],[380,0]]
       }
       CTRL_L_mouth_cornerSharpnessU {type=MovieSceneScriptingFloatChannel, default=0, const [0,3]}
       CTRL_R_mouth_lipsTowardsTeethD {type=MovieSceneScriptingFloatChannel, default=0, const [0,4]}
       CTRL_L_mouth_lipsTowardsTeethD {type=MovieSceneScriptingFloatChannel, default=0, const [0,4]}
       CTRL_R_eye_eyelidU.X {
        type=MovieSceneScriptingFloatChannel
        default=0
        thinned_from=28
        keys [[0,0],[92,0],[99,-0.0747],[114,-0.0436],[122,-0.0391],[127,-0.0062],[130,-0.0015],[135,0],[155,0],[163,0.0237],[165,0.0154],[167,-0.0031],[175,-0.1907],[180,0.3069],[186,0.9611],[266,0.9611],[275,-0.2241],[278,-0.2305],[285,-0.1082],[288,-0.1071],[305,-0.1032],[309,-0.1227],[318,-0.0979],[350,-0.0935],[369,-0.0117],[380,0]]
       }
       CTRL_R_eye_eyelidU.Y {
        type=MovieSceneScriptingFloatChannel
        default=0
        thinned_from=28
        keys [[0,0],[92,0],[99,-0.3726],[114,-0.2183],[122,-0.205],[127,-0.0326],[130,-0.0079],[135,0],[155,0],[163,-0.0773],[164,-0.0241],[165,0.0162],[167,-0.0481],[175,0.6578],[180,0.3244],[186,0.3871],[266,1],[275,-0.8343],[278,-0.9252],[285,0.3638],[288,0.5454],[305,0.5357],[309,0.0485],[318,0.3998],[350,0.4715],[369,0.0589],[380,0]]
       }
       CTRL_L_mouth_lipsTowardsTeethU {type=MovieSceneScriptingFloatChannel, default=0, const [0,4]}
       CTRL_R_eye_eyelidD.X {
        type=MovieSceneScriptingFloatChannel
        default=0
        thinned_from=22
        keys [[0,0],[142,0],[159,-0.3401],[160,-0.3624],[165,-0.4841],[169,-0.5564],[175,-0.6007],[186,-0.0769],[266,-0.0769],[274,-0.7811],[275,-0.8061],[276,-0.791],[277,-0.7733],[279,-0.7188],[296,-0.1492],[350,0],[380,0]]
       }
       CTRL_R_eye_eyelidD.Y {
        type=MovieSceneScriptingFloatChannel
        default=0
        thinned_from=20
        keys [[0,0],[142,0],[159,0.1034],[161,-0.2154],[166,-0.1233],[169,0.6109],[175,0.6595],[186,0.3449],[266,0.3449],[274,0.1324],[276,-0.3315],[282,-0.6062],[296,0.1638],[350,0],[380,0]]
       }
       CTRL_R_mouth_lipsTowardsTeethU {type=MovieSceneScriptingFloatChannel, default=0, const [0,4]}
       CTRL_faceGUIfollowHead {type=MovieSceneScriptingFloatChannel, default=1}
       CTRL_eyesAimFollowHead {type=MovieSceneScriptingFloatChannel, default=0}
       CTRL_C_nose.X {type=MovieSceneScriptingFloatChannel, default=0, const [0,4]}
       CTRL_C_nose.Y {type=MovieSceneScriptingFloatChannel, default=0, const [0,4]}
       CTRL_C_teeth_rotateU {type=MovieSceneScriptingFloatChannel, default=0, const [0,3]}
       CTRL_C_teeth_rotateD {type=MovieSceneScriptingFloatChannel, default=0, const [0,3]}
       CTRL_L_brow_inner {type=MovieSceneScriptingFloatChannel, default=0, const [0,2]}
       CTRL_R_brow_inner {type=MovieSceneScriptingFloatChannel, default=0, const [0,2]}
       CTRL_L_brow_outer {type=MovieSceneScriptingFloatChannel, default=0, const [0,2]}
       CTRL_R_brow_outer {type=MovieSceneScriptingFloatChannel, default=0, const [0,2]}
       CTRL_C_mouth.X {type=MovieSceneScriptingFloatChannel, default=-1}
       CTRL_C_mouth.Y {type=MovieSceneScriptingFloatChannel, default=0.5386}
       CTRL_L_mouth_upperLipRaise {type=MovieSceneScriptingFloatChannel, default=0}
       CTRL_L_mouth_cornerPull {type=MovieSceneScriptingFloatChannel, default=0}
       CTRL_L_mouth_dimple {type=MovieSceneScriptingFloatChannel, default=0}
       CTRL_L_mouth_cornerDepress {type=MovieSceneScriptingFloatChannel, default=0}
       CTRL_L_mouth_stretch {type=MovieSceneScriptingFloatChannel, default=0}
       CTRL_L_mouth_lowerLipDepress {type=MovieSceneScriptingFloatChannel, default=0}
       CTRL_R_mouth_lowerLipDepress {type=MovieSceneScriptingFloatChannel, default=0}
       CTRL_R_mouth_stretch {type=MovieSceneScriptingFloatChannel, default=0}
       CTRL_L_mouth_towardsD {type=MovieSceneScriptingFloatChannel, default=0}
       CTRL_R_mouth_funnelU {type=MovieSceneScriptingFloatChannel, default=0}
       CTRL_L_mouth_towardsU {type=MovieSceneScriptingFloatChannel, default=0}
       CTRL_R_mouth_lipBiteD {type=MovieSceneScriptingFloatChannel, default=0}
       CTRL_R_mouth_funnelD {type=MovieSceneScriptingFloatChannel, default=0}
       CTRL_L_mouth_pressU {type=MovieSceneScriptingFloatChannel, default=0}
       CTRL_L_mouth_funnelD {type=MovieSceneScriptingFloatChannel, default=0}
       CTRL_L_mouth_suckBlow {type=MovieSceneScriptingFloatChannel, default=0, const [0,3]}
       CTRL_L_mouth_pressD {type=MovieSceneScriptingFloatChannel, default=0}
       CTRL_L_mouth_funnelU {type=MovieSceneScriptingFloatChannel, default=0}
       CTRL_R_mouth_towardsU {type=MovieSceneScriptingFloatChannel, default=0}
       CTRL_R_mouth_purseD {type=MovieSceneScriptingFloatChannel, default=0}
       CTRL_L_mouth_lipBiteU {type=MovieSceneScriptingFloatChannel, default=0}
       CTRL_R_mouth_pressD {type=MovieSceneScriptingFloatChannel, default=0}
       CTRL_R_mouth_pressU {type=MovieSceneScriptingFloatChannel, default=0}
       CTRL_L_mouth_purseD {type=MovieSceneScriptingFloatChannel, default=0}
       CTRL_R_mouth_lipBiteU {type=MovieSceneScriptingFloatChannel, default=0}
       CTRL_R_mouth_suckBlow {type=MovieSceneScriptingFloatChannel, default=0, const [0,3]}
       CTRL_L_mouth_purseU {type=MovieSceneScriptingFloatChannel, default=0}
       CTRL_R_mouth_towardsD {type=MovieSceneScriptingFloatChannel, default=0}
       CTRL_L_mouth_lipBiteD {type=MovieSceneScriptingFloatChannel, default=0}
       CTRL_R_mouth_purseU {type=MovieSceneScriptingFloatChannel, default=0}
       CTRL_R_mouth_cornerDepress {type=MovieSceneScriptingFloatChannel, default=0.3417}
       CTRL_R_mouth_upperLipRaise {type=MovieSceneScriptingFloatChannel, default=0}
       CTRL_R_mouth_dimple {type=MovieSceneScriptingFloatChannel, default=0, keys [[126,0],[132,0],[137,1]]}
       CTRL_R_mouth_cornerPull {type=MovieSceneScriptingFloatChannel, default=0}
      ]
     }]
   }
   {
    type=MovieSceneControlRigParameterTrack
    display_name="Fortnite_Biped_CtrlRig (Layered)"
    sections [{
      type=MovieSceneControlRigParameterSection
      channels [
       leg_l_stretch_switch {type=MovieSceneScriptingBoolChannel, default=false}
       leg_r_stretch_switch {type=MovieSceneScriptingBoolChannel, default=false}
       arm_l_fk_local {type=MovieSceneScriptingBoolChannel, default=true}
       Auto Interpolate_3 {type=MovieSceneScriptingBoolChannel, default=false}
       Auto Interpolate {type=MovieSceneScriptingBoolChannel, default=false}
       arm_l_stretch_switch {type=MovieSceneScriptingBoolChannel, default=false}
       Keep Aligned {type=MovieSceneScriptingBoolChannel, default=false}
       arm_r_fk_local {type=MovieSceneScriptingBoolChannel, default=true}
       Auto Interpolate_4 {type=MovieSceneScriptingBoolChannel, default=false}
       Auto Interpolate_2 {type=MovieSceneScriptingBoolChannel, default=false}
       arm_r_stretch_switch {type=MovieSceneScriptingBoolChannel, default=false}
       Keep Aligned_2 {type=MovieSceneScriptingBoolChannel, default=false}
       head_local {type=MovieSceneScriptingBoolChannel, default=true}
       Stretch {type=MovieSceneScriptingBoolChannel, default=false}
       head_stretch {type=MovieSceneScriptingBoolChannel, default=false}
       spine_stretch_switch {type=MovieSceneScriptingBoolChannel, default=false}
       arm_l_fk_ik_switch {type=MovieSceneScriptingBoolChannel, default=false}
       leg_l_fk_ik_switch {type=MovieSceneScriptingBoolChannel, default=true}
       arm_r_fk_ik_switch {type=MovieSceneScriptingBoolChannel, default=false}
       leg_r_fk_ik_switch {type=MovieSceneScriptingBoolChannel, default=true}
       ShowBodyControls {type=MovieSceneScriptingBoolChannel, default=true}
       spine_fk_ik_switch {type=MovieSceneScriptingBoolChannel, default=false}
       neck_fk_ik_switch {type=MovieSceneScriptingBoolChannel, default=false}
       space {type=MovieSceneScriptingByteChannel, default=1}
       global_ctrl.Location.X {type=MovieSceneScriptingFloatChannel, default=0}
       global_ctrl.Location.Y {type=MovieSceneScriptingFloatChannel, default=0}
       global_ctrl.Location.Z {type=MovieSceneScriptingFloatChannel, default=0}
       global_ctrl.Rotation.X {type=MovieSceneScriptingFloatChannel, default=0}
       global_ctrl.Rotation.Y {type=MovieSceneScriptingFloatChannel, default=0}
       global_ctrl.Rotation.Z {type=MovieSceneScriptingFloatChannel, default=0}
       root_ctrl.Location.X {type=MovieSceneScriptingFloatChannel, default=0}
       root_ctrl.Location.Y {type=MovieSceneScriptingFloatChannel, default=0}
       root_ctrl.Location.Z {type=MovieSceneScriptingFloatChannel, default=0}
       root_ctrl.Rotation.X {type=MovieSceneScriptingFloatChannel, default=0}
       root_ctrl.Rotation.Y {type=MovieSceneScriptingFloatChannel, default=0}
       root_ctrl.Rotation.Z {type=MovieSceneScriptingFloatChannel, default=0}
       root_ctrl.Scale.X {type=MovieSceneScriptingFloatChannel, default=1}
       root_ctrl.Scale.Y {type=MovieSceneScriptingFloatChannel, default=1}
       root_ctrl.Scale.Z {type=MovieSceneScriptingFloatChannel, default=1}
       body_offset_ctrl.Location.X {type=MovieSceneScriptingFloatChannel, default=0}
       body_offset_ctrl.Location.Y {type=MovieSceneScriptingFloatChannel, default=0}
       body_offset_ctrl.Location.Z {type=MovieSceneScriptingFloatChannel, default=0}
       body_offset_ctrl.Rotation.X {type=MovieSceneScriptingFloatChannel, default=0}
       body_offset_ctrl.Rotation.Y {type=MovieSceneScriptingFloatChannel, default=0}
       body_offset_ctrl.Rotation.Z {type=MovieSceneScriptingFloatChannel, default=0}
       body_offset_ctrl.Scale.X {type=MovieSceneScriptingFloatChannel, default=1}
       body_offset_ctrl.Scale.Y {type=MovieSceneScriptingFloatChannel, default=1}
       body_offset_ctrl.Scale.Z {type=MovieSceneScriptingFloatChannel, default=1}
       thigh_l_fk_ctrl.Location.X {type=MovieSceneScriptingFloatChannel, default=0}
       thigh_l_fk_ctrl.Location.Y {type=MovieSceneScriptingFloatChannel, default=0}
       thigh_l_fk_ctrl.Location.Z {type=MovieSceneScriptingFloatChannel, default=0}
       thigh_l_fk_ctrl.Rotation.X {type=MovieSceneScriptingFloatChannel, default=-0.0445}
       thigh_l_fk_ctrl.Rotation.Y {type=MovieSceneScriptingFloatChannel, default=-0.0019}
       thigh_l_fk_ctrl.Rotation.Z {type=MovieSceneScriptingFloatChannel, default=0}
       thigh_l_fk_ctrl.Scale.X {type=MovieSceneScriptingFloatChannel, default=1}
       thigh_l_fk_ctrl.Scale.Y {type=MovieSceneScriptingFloatChannel, default=1}
       thigh_l_fk_ctrl.Scale.Z {type=MovieSceneScriptingFloatChannel, default=1}
       calf_l_fk_ctrl.Location.X {type=MovieSceneScriptingFloatChannel, default=0}
       calf_l_fk_ctrl.Location.Y {type=MovieSceneScriptingFloatChannel, default=0}
       calf_l_fk_ctrl.Location.Z {type=MovieSceneScriptingFloatChannel, default=0}
       calf_l_fk_ctrl.Rotation.X {type=MovieSceneScriptingFloatChannel, default=0}
       calf_l_fk_ctrl.Rotation.Y {type=MovieSceneScriptingFloatChannel, default=0}
       calf_l_fk_ctrl.Rotation.Z {type=MovieSceneScriptingFloatChannel, default=0}
       calf_l_fk_ctrl.Scale.X {type=MovieSceneScriptingFloatChannel, default=1}
       calf_l_fk_ctrl.Scale.Y {type=MovieSceneScriptingFloatChannel, default=1}
       calf_l_fk_ctrl.Scale.Z {type=MovieSceneScriptingFloatChannel, default=1}
       foot_l_fk_ctrl.Location.X {type=MovieSceneScriptingFloatChannel, default=0}
       foot_l_fk_ctrl.Location.Y {type=MovieSceneScriptingFloatChannel, default=0}
       foot_l_fk_ctrl.Location.Z {type=MovieSceneScriptingFloatChannel, default=0}
       foot_l_fk_ctrl.Rotation.X {type=MovieSceneScriptingFloatChannel, default=0.0445}
       foot_l_fk_ctrl.Rotation.Y {type=MovieSceneScriptingFloatChannel, default=-0.0016}
       foot_l_fk_ctrl.Rotation.Z {type=MovieSceneScriptingFloatChannel, default=0}
       foot_l_fk_ctrl.Scale.X {type=MovieSceneScriptingFloatChannel, default=1}
       foot_l_fk_ctrl.Scale.Y {type=MovieSceneScriptingFloatChannel, default=1}
       foot_l_fk_ctrl.Scale.Z {type=MovieSceneScriptingFloatChannel, default=1}
       ball_l_fk_ctrl.Location.X {type=MovieSceneScriptingFloatChannel, default=0}
       ball_l_fk_ctrl.Location.Y {type=MovieSceneScriptingFloatChannel, default=0}
       ball_l_fk_ctrl.Location.Z {type=MovieSceneScriptingFloatChannel, default=0}
       ball_l_fk_ctrl.Rotation.X {type=MovieSceneScriptingFloatChannel, default=0}
       ball_l_fk_ctrl.Rotation.Y {type=MovieSceneScriptingFloatChannel, default=0}
       ball_l_fk_ctrl.Rotation.Z {type=MovieSceneScriptingFloatChannel, default=0}
       ball_l_fk_ctrl.Scale.X {type=MovieSceneScriptingFloatChannel, default=1}
       ball_l_fk_ctrl.Scale.Y {type=MovieSceneScriptingFloatChannel, default=1}
       ball_l_fk_ctrl.Scale.Z {type=MovieSceneScriptingFloatChannel, default=1}
       foot_l_ik_ctrl.Location.X {type=MovieSceneScriptingFloatChannel, default=0}
       foot_l_ik_ctrl.Location.Y {type=MovieSceneScriptingFloatChannel, default=0}
       foot_l_ik_ctrl.Location.Z {type=MovieSceneScriptingFloatChannel, default=0}
       foot_l_ik_ctrl.Rotation.X {type=MovieSceneScriptingFloatChannel, default=0}
       foot_l_ik_ctrl.Rotation.Y {type=MovieSceneScriptingFloatChannel, default=0}
       foot_l_ik_ctrl.Rotation.Z {type=MovieSceneScriptingFloatChannel, default=0}
       foot_l_ik_ctrl.Scale.X {type=MovieSceneScriptingFloatChannel, default=1}
       foot_l_ik_ctrl.Scale.Y {type=MovieSceneScriptingFloatChannel, default=1}
       foot_l_ik_ctrl.Scale.Z {type=MovieSceneScriptingFloatChannel, default=1}
       ball_l_ik_ctrl.Location.X {type=MovieSceneScriptingFloatChannel, default=0}
       ball_l_ik_ctrl.Location.Y {type=MovieSceneScriptingFloatChannel, default=0}
       ball_l_ik_ctrl.Location.Z {type=MovieSceneScriptingFloatChannel, default=0}
       ball_l_ik_ctrl.Rotation.X {type=MovieSceneScriptingFloatChannel, default=0}
       ball_l_ik_ctrl.Rotation.Y {type=MovieSceneScriptingFloatChannel, default=0}
       ball_l_ik_ctrl.Rotation.Z {type=MovieSceneScriptingFloatChannel, default=0}
       ball_l_ik_ctrl.Scale.X {type=MovieSceneScriptingFloatChannel, default=1}
       ball_l_ik_ctrl.Scale.Y {type=MovieSceneScriptingFloatChannel, default=1}
       ball_l_ik_ctrl.Scale.Z {type=MovieSceneScriptingFloatChannel, default=1}
       foot_bk1_l_ctrl.Location.X {type=MovieSceneScriptingFloatChannel, default=0}
       foot_bk1_l_ctrl.Location.Y {type=MovieSceneScriptingFloatChannel, default=0}
       foot_bk1_l_ctrl.Location.Z {type=MovieSceneScriptingFloatChannel, default=0}
       foot_bk1_l_ctrl.Rotation.X {type=MovieSceneScriptingFloatChannel, default=0}
       foot_bk1_l_ctrl.Rotation.Y {type=MovieSceneScriptingFloatChannel, default=0}
       foot_bk1_l_ctrl.Rotation.Z {type=MovieSceneScriptingFloatChannel, default=0}
       foot_bk1_l_ctrl.Scale.X {type=MovieSceneScriptingFloatChannel, default=1}
       foot_bk1_l_ctrl.Scale.Y {type=MovieSceneScriptingFloatChannel, default=1}
       foot_bk1_l_ctrl.Scale.Z {type=MovieSceneScriptingFloatChannel, default=1}
       foot_roll_l_ctrl.Location.X {type=MovieSceneScriptingFloatChannel, default=0}
       foot_roll_l_ctrl.Location.Y {type=MovieSceneScriptingFloatChannel, default=0}
       foot_roll_l_ctrl.Location.Z {type=MovieSceneScriptingFloatChannel, default=0}
       foot_roll_l_ctrl.Rotation.X {type=MovieSceneScriptingFloatChannel, default=0}
       foot_roll_l_ctrl.Rotation.Y {type=MovieSceneScriptingFloatChannel, default=0}
       foot_roll_l_ctrl.Rotation.Z {type=MovieSceneScriptingFloatChannel, default=0}
       foot_roll_l_ctrl.Scale.X {type=MovieSceneScriptingFloatChannel, default=1}
       foot_roll_l_ctrl.Scale.Y {type=MovieSceneScriptingFloatChannel, default=1}
       foot_roll_l_ctrl.Scale.Z {type=MovieSceneScriptingFloatChannel, default=1}
       roll_blend_l_ctrl {type=MovieSceneScriptingFloatChannel, default=0}
       tip_l_ctrl.Location.X {type=MovieSceneScriptingFloatChannel, default=0}
       tip_l_ctrl.Location.Y {type=MovieSceneScriptingFloatChannel, default=0}
       tip_l_ctrl.Location.Z {type=MovieSceneScriptingFloatChannel, default=0}
       tip_l_ctrl.Rotation.X {type=MovieSceneScriptingFloatChannel, default=0}
       tip_l_ctrl.Rotation.Y {type=MovieSceneScriptingFloatChannel, default=0}
       tip_l_ctrl.Rotation.Z {type=MovieSceneScriptingFloatChannel, default=0}
       tip_l_ctrl.Scale.X {type=MovieSceneScriptingFloatChannel, default=1}
       tip_l_ctrl.Scale.Y {type=MovieSceneScriptingFloatChannel, default=1}
       tip_l_ctrl.Scale.Z {type=MovieSceneScriptingFloatChannel, default=1}
       heel_l_ctrl.Location.X {type=MovieSceneScriptingFloatChannel, default=0}
       heel_l_ctrl.Location.Y {type=MovieSceneScriptingFloatChannel, default=0}
       heel_l_ctrl.Location.Z {type=MovieSceneScriptingFloatChannel, default=0}
       heel_l_ctrl.Rotation.X {type=MovieSceneScriptingFloatChannel, default=0}
       heel_l_ctrl.Rotation.Y {type=MovieSceneScriptingFloatChannel, default=0}
       heel_l_ctrl.Rotation.Z {type=MovieSceneScriptingFloatChannel, default=0}
       heel_l_ctrl.Scale.X {type=MovieSceneScriptingFloatChannel, default=1}
       heel_l_ctrl.Scale.Y {type=MovieSceneScriptingFloatChannel, default=1}
       heel_l_ctrl.Scale.Z {type=MovieSceneScriptingFloatChannel, default=1}
       leg_l_softness {type=MovieSceneScriptingFloatChannel, default=0}
       leg_l_pv_ik_ctrl.Location.X {type=MovieSceneScriptingFloatChannel, default=0}
       leg_l_pv_ik_ctrl.Location.Y {type=MovieSceneScriptingFloatChannel, default=0}
       leg_l_pv_ik_ctrl.Location.Z {type=MovieSceneScriptingFloatChannel, default=0}
       leg_l_pv_ik_ctrl.Rotation.X {type=MovieSceneScriptingFloatChannel, default=0}
       leg_l_pv_ik_ctrl.Rotation.Y {type=MovieSceneScriptingFloatChannel, default=0}
       leg_l_pv_ik_ctrl.Rotation.Z {type=MovieSceneScriptingFloatChannel, default=0}
       leg_l_pv_ik_ctrl.Scale.X {type=MovieSceneScriptingFloatChannel, default=1}
       leg_l_pv_ik_ctrl.Scale.Y {type=MovieSceneScriptingFloatChannel, default=1}
       leg_l_pv_ik_ctrl.Scale.Z {type=MovieSceneScriptingFloatChannel, default=1}
       thigh_r_fk_ctrl.Location.X {type=MovieSceneScriptingFloatChannel, default=0}
       thigh_r_fk_ctrl.Location.Y {type=MovieSceneScriptingFloatChannel, default=0}
       thigh_r_fk_ctrl.Location.Z {type=MovieSceneScriptingFloatChannel, default=0}
       thigh_r_fk_ctrl.Rotation.X {type=MovieSceneScriptingFloatChannel, default=-0.0486}
       thigh_r_fk_ctrl.Rotation.Y {type=MovieSceneScriptingFloatChannel, default=-0.0021}
       thigh_r_fk_ctrl.Rotation.Z {type=MovieSceneScriptingFloatChannel, default=0}
       thigh_r_fk_ctrl.Scale.X {type=MovieSceneScriptingFloatChannel, default=1}
       thigh_r_fk_ctrl.Scale.Y {type=MovieSceneScriptingFloatChannel, default=1}
       thigh_r_fk_ctrl.Scale.Z {type=MovieSceneScriptingFloatChannel, default=1}
       calf_r_fk_ctrl.Location.X {type=MovieSceneScriptingFloatChannel, default=0}
       calf_r_fk_ctrl.Location.Y {type=MovieSceneScriptingFloatChannel, default=0}
       calf_r_fk_ctrl.Location.Z {type=MovieSceneScriptingFloatChannel, default=0}
       calf_r_fk_ctrl.Rotation.X {type=MovieSceneScriptingFloatChannel, default=0}
       calf_r_fk_ctrl.Rotation.Y {type=MovieSceneScriptingFloatChannel, default=0}
       calf_r_fk_ctrl.Rotation.Z {type=MovieSceneScriptingFloatChannel, default=0}
       calf_r_fk_ctrl.Scale.X {type=MovieSceneScriptingFloatChannel, default=1}
       calf_r_fk_ctrl.Scale.Y {type=MovieSceneScriptingFloatChannel, default=1}
       calf_r_fk_ctrl.Scale.Z {type=MovieSceneScriptingFloatChannel, default=1}
       foot_r_fk_ctrl.Location.X {type=MovieSceneScriptingFloatChannel, default=0}
       foot_r_fk_ctrl.Location.Y {type=MovieSceneScriptingFloatChannel, default=0}
       foot_r_fk_ctrl.Location.Z {type=MovieSceneScriptingFloatChannel, default=0}
       foot_r_fk_ctrl.Rotation.X {type=MovieSceneScriptingFloatChannel, default=0.0486}
       foot_r_fk_ctrl.Rotation.Y {type=MovieSceneScriptingFloatChannel, default=-0.0018}
       foot_r_fk_ctrl.Rotation.Z {type=MovieSceneScriptingFloatChannel, default=0.0001}
       foot_r_fk_ctrl.Scale.X {type=MovieSceneScriptingFloatChannel, default=1}
       foot_r_fk_ctrl.Scale.Y {type=MovieSceneScriptingFloatChannel, default=1}
       foot_r_fk_ctrl.Scale.Z {type=MovieSceneScriptingFloatChannel, default=1}
       ball_r_fk_ctrl.Location.X {type=MovieSceneScriptingFloatChannel, default=0}
       ball_r_fk_ctrl.Location.Y {type=MovieSceneScriptingFloatChannel, default=0}
       ball_r_fk_ctrl.Location.Z {type=MovieSceneScriptingFloatChannel, default=0}
       ball_r_fk_ctrl.Rotation.X {type=MovieSceneScriptingFloatChannel, default=0}
       ball_r_fk_ctrl.Rotation.Y {type=MovieSceneScriptingFloatChannel, default=0}
       ball_r_fk_ctrl.Rotation.Z {type=MovieSceneScriptingFloatChannel, default=0}
       ball_r_fk_ctrl.Scale.X {type=MovieSceneScriptingFloatChannel, default=1}
       ball_r_fk_ctrl.Scale.Y {type=MovieSceneScriptingFloatChannel, default=1}
       ball_r_fk_ctrl.Scale.Z {type=MovieSceneScriptingFloatChannel, default=1}
       foot_r_ik_ctrl.Location.X {type=MovieSceneScriptingFloatChannel, default=0}
       foot_r_ik_ctrl.Location.Y {type=MovieSceneScriptingFloatChannel, default=0}
       foot_r_ik_ctrl.Location.Z {type=MovieSceneScriptingFloatChannel, default=0}
       foot_r_ik_ctrl.Rotation.X {type=MovieSceneScriptingFloatChannel, default=0}
       foot_r_ik_ctrl.Rotation.Y {type=MovieSceneScriptingFloatChannel, default=0}
       foot_r_ik_ctrl.Rotation.Z {type=MovieSceneScriptingFloatChannel, default=0}
       foot_r_ik_ctrl.Scale.X {type=MovieSceneScriptingFloatChannel, default=1}
       foot_r_ik_ctrl.Scale.Y {type=MovieSceneScriptingFloatChannel, default=1}
       foot_r_ik_ctrl.Scale.Z {type=MovieSceneScriptingFloatChannel, default=1}
       ball_r_ik_ctrl.Location.X {type=MovieSceneScriptingFloatChannel, default=0}
       ball_r_ik_ctrl.Location.Y {type=MovieSceneScriptingFloatChannel, default=0}
       ball_r_ik_ctrl.Location.Z {type=MovieSceneScriptingFloatChannel, default=0}
       ball_r_ik_ctrl.Rotation.X {type=MovieSceneScriptingFloatChannel, default=0}
       ball_r_ik_ctrl.Rotation.Y {type=MovieSceneScriptingFloatChannel, default=0}
       ball_r_ik_ctrl.Rotation.Z {type=MovieSceneScriptingFloatChannel, default=0}
       ball_r_ik_ctrl.Scale.X {type=MovieSceneScriptingFloatChannel, default=1}
       ball_r_ik_ctrl.Scale.Y {type=MovieSceneScriptingFloatChannel, default=1}
       ball_r_ik_ctrl.Scale.Z {type=MovieSceneScriptingFloatChannel, default=1}
       foot_bk1_r_ctrl.Location.X {type=MovieSceneScriptingFloatChannel, default=0}
       foot_bk1_r_ctrl.Location.Y {type=MovieSceneScriptingFloatChannel, default=0}
       foot_bk1_r_ctrl.Location.Z {type=MovieSceneScriptingFloatChannel, default=0}
       foot_bk1_r_ctrl.Rotation.X {type=MovieSceneScriptingFloatChannel, default=0}
       foot_bk1_r_ctrl.Rotation.Y {type=MovieSceneScriptingFloatChannel, default=0}
       foot_bk1_r_ctrl.Rotation.Z {type=MovieSceneScriptingFloatChannel, default=0}
       foot_bk1_r_ctrl.Scale.X {type=MovieSceneScriptingFloatChannel, default=1}
       foot_bk1_r_ctrl.Scale.Y {type=MovieSceneScriptingFloatChannel, default=1}
       foot_bk1_r_ctrl.Scale.Z {type=MovieSceneScriptingFloatChannel, default=1}
       foot_roll_r_ctrl.Location.X {type=MovieSceneScriptingFloatChannel, default=0}
       foot_roll_r_ctrl.Location.Y {type=MovieSceneScriptingFloatChannel, default=0}
       foot_roll_r_ctrl.Location.Z {type=MovieSceneScriptingFloatChannel, default=0}
       foot_roll_r_ctrl.Rotation.X {type=MovieSceneScriptingFloatChannel, default=0}
       foot_roll_r_ctrl.Rotation.Y {type=MovieSceneScriptingFloatChannel, default=0}
       foot_roll_r_ctrl.Rotation.Z {type=MovieSceneScriptingFloatChannel, default=0}
       foot_roll_r_ctrl.Scale.X {type=MovieSceneScriptingFloatChannel, default=1}
       foot_roll_r_ctrl.Scale.Y {type=MovieSceneScriptingFloatChannel, default=1}
       foot_roll_r_ctrl.Scale.Z {type=MovieSceneScriptingFloatChannel, default=1}
       roll_blend_r_ctrl {type=MovieSceneScriptingFloatChannel, default=0}
       tip_r_ctrl.Location.X {type=MovieSceneScriptingFloatChannel, default=0}
       tip_r_ctrl.Location.Y {type=MovieSceneScriptingFloatChannel, default=0}
       tip_r_ctrl.Location.Z {type=MovieSceneScriptingFloatChannel, default=0}
       tip_r_ctrl.Rotation.X {type=MovieSceneScriptingFloatChannel, default=0}
       tip_r_ctrl.Rotation.Y {type=MovieSceneScriptingFloatChannel, default=0}
       tip_r_ctrl.Rotation.Z {type=MovieSceneScriptingFloatChannel, default=0}
       tip_r_ctrl.Scale.X {type=MovieSceneScriptingFloatChannel, default=1}
       tip_r_ctrl.Scale.Y {type=MovieSceneScriptingFloatChannel, default=1}
       tip_r_ctrl.Scale.Z {type=MovieSceneScriptingFloatChannel, default=1}
       heel_r_ctrl.Location.X {type=MovieSceneScriptingFloatChannel, default=0}
       heel_r_ctrl.Location.Y {type=MovieSceneScriptingFloatChannel, default=0}
       heel_r_ctrl.Location.Z {type=MovieSceneScriptingFloatChannel, default=0}
       heel_r_ctrl.Rotation.X {type=MovieSceneScriptingFloatChannel, default=0}
       heel_r_ctrl.Rotation.Y {type=MovieSceneScriptingFloatChannel, default=0}
       heel_r_ctrl.Rotation.Z {type=MovieSceneScriptingFloatChannel, default=0}
       heel_r_ctrl.Scale.X {type=MovieSceneScriptingFloatChannel, default=1}
       heel_r_ctrl.Scale.Y {type=MovieSceneScriptingFloatChannel, default=1}
       heel_r_ctrl.Scale.Z {type=MovieSceneScriptingFloatChannel, default=1}
       leg_r_softness {type=MovieSceneScriptingFloatChannel, default=0}
       leg_r_pv_ik_ctrl.Location.X {type=MovieSceneScriptingFloatChannel, default=0}
       leg_r_pv_ik_ctrl.Location.Y {type=MovieSceneScriptingFloatChannel, default=0}
       leg_r_pv_ik_ctrl.Location.Z {type=MovieSceneScriptingFloatChannel, default=0}
       leg_r_pv_ik_ctrl.Rotation.X {type=MovieSceneScriptingFloatChannel, default=0}
       leg_r_pv_ik_ctrl.Rotation.Y {type=MovieSceneScriptingFloatChannel, default=0}
       leg_r_pv_ik_ctrl.Rotation.Z {type=MovieSceneScriptingFloatChannel, default=0}
       leg_r_pv_ik_ctrl.Scale.X {type=MovieSceneScriptingFloatChannel, default=1}
       leg_r_pv_ik_ctrl.Scale.Y {type=MovieSceneScriptingFloatChannel, default=1}
       leg_r_pv_ik_ctrl.Scale.Z {type=MovieSceneScriptingFloatChannel, default=1}
       upperarm_l_fk_ctrl.Location.X {type=MovieSceneScriptingFloatChannel, default=0}
       upperarm_l_fk_ctrl.Location.Y {type=MovieSceneScriptingFloatChannel, default=0}
       upperarm_l_fk_ctrl.Location.Z {type=MovieSceneScriptingFloatChannel, default=0}
       upperarm_l_fk_ctrl.Rotation.X {type=MovieSceneScriptingFloatChannel, default=0}
       upperarm_l_fk_ctrl.Rotation.Y {type=MovieSceneScriptingFloatChannel, default=0}
       upperarm_l_fk_ctrl.Rotation.Z {type=MovieSceneScriptingFloatChannel, default=0}
       upperarm_l_fk_ctrl.Scale.X {type=MovieSceneScriptingFloatChannel, default=1}
       upperarm_l_fk_ctrl.Scale.Y {type=MovieSceneScriptingFloatChannel, default=1}
       upperarm_l_fk_ctrl.Scale.Z {type=MovieSceneScriptingFloatChannel, default=1}
       lowerarm_l_fk_ctrl.Location.X {type=MovieSceneScriptingFloatChannel, default=0}
       lowerarm_l_fk_ctrl.Location.Y {type=MovieSceneScriptingFloatChannel, default=0}
       lowerarm_l_fk_ctrl.Location.Z {type=MovieSceneScriptingFloatChannel, default=0}
       lowerarm_l_fk_ctrl.Rotation.X {type=MovieSceneScriptingFloatChannel, default=0}
       lowerarm_l_fk_ctrl.Rotation.Y {type=MovieSceneScriptingFloatChannel, default=0}
       lowerarm_l_fk_ctrl.Rotation.Z {type=MovieSceneScriptingFloatChannel, default=0}
       lowerarm_l_fk_ctrl.Scale.X {type=MovieSceneScriptingFloatChannel, default=1}
       lowerarm_l_fk_ctrl.Scale.Y {type=MovieSceneScriptingFloatChannel, default=1}
       lowerarm_l_fk_ctrl.Scale.Z {type=MovieSceneScriptingFloatChannel, default=1}
       hand_l_fk_ctrl.Location.X {type=MovieSceneScriptingFloatChannel, default=0}
       hand_l_fk_ctrl.Location.Y {type=MovieSceneScriptingFloatChannel, default=0}
       hand_l_fk_ctrl.Location.Z {type=MovieSceneScriptingFloatChannel, default=0}
       hand_l_fk_ctrl.Rotation.X {type=MovieSceneScriptingFloatChannel, default=0}
       hand_l_fk_ctrl.Rotation.Y {type=MovieSceneScriptingFloatChannel, default=0}
       hand_l_fk_ctrl.Rotation.Z {type=MovieSceneScriptingFloatChannel, default=0}
       hand_l_fk_ctrl.Scale.X {type=MovieSceneScriptingFloatChannel, default=1}
       hand_l_fk_ctrl.Scale.Y {type=MovieSceneScriptingFloatChannel, default=1}
       hand_l_fk_ctrl.Scale.Z {type=MovieSceneScriptingFloatChannel, default=1}
       thumb_01_l_ctrl.Location.X {type=MovieSceneScriptingFloatChannel, default=0}
       thumb_01_l_ctrl.Location.Y {type=MovieSceneScriptingFloatChannel, default=0}
       thumb_01_l_ctrl.Location.Z {type=MovieSceneScriptingFloatChannel, default=0}
       thumb_01_l_ctrl.Rotation.X {type=MovieSceneScriptingFloatChannel, default=0}
       thumb_01_l_ctrl.Rotation.Y {type=MovieSceneScriptingFloatChannel, default=0}
       thumb_01_l_ctrl.Rotation.Z {type=MovieSceneScriptingFloatChannel, default=0}
       thumb_01_l_ctrl.Scale.X {type=MovieSceneScriptingFloatChannel, default=1}
       thumb_01_l_ctrl.Scale.Y {type=MovieSceneScriptingFloatChannel, default=1}
       thumb_01_l_ctrl.Scale.Z {type=MovieSceneScriptingFloatChannel, default=1}
       thumb_02_l_ctrl.Location.X {type=MovieSceneScriptingFloatChannel, default=0}
       thumb_02_l_ctrl.Location.Y {type=MovieSceneScriptingFloatChannel, default=0}
       thumb_02_l_ctrl.Location.Z {type=MovieSceneScriptingFloatChannel, default=0}
       thumb_02_l_ctrl.Rotation.X {type=MovieSceneScriptingFloatChannel, default=0}
       thumb_02_l_ctrl.Rotation.Y {type=MovieSceneScriptingFloatChannel, default=0}
       thumb_02_l_ctrl.Rotation.Z {type=MovieSceneScriptingFloatChannel, default=0}
       thumb_02_l_ctrl.Scale.X {type=MovieSceneScriptingFloatChannel, default=1}
       thumb_02_l_ctrl.Scale.Y {type=MovieSceneScriptingFloatChannel, default=1}
       thumb_02_l_ctrl.Scale.Z {type=MovieSceneScriptingFloatChannel, default=1}
       thumb_03_l_ctrl.Location.X {type=MovieSceneScriptingFloatChannel, default=0}
       thumb_03_l_ctrl.Location.Y {type=MovieSceneScriptingFloatChannel, default=0}
       thumb_03_l_ctrl.Location.Z {type=MovieSceneScriptingFloatChannel, default=0}
       thumb_03_l_ctrl.Rotation.X {type=MovieSceneScriptingFloatChannel, default=0}
       thumb_03_l_ctrl.Rotation.Y {type=MovieSceneScriptingFloatChannel, default=0}
       thumb_03_l_ctrl.Rotation.Z {type=MovieSceneScriptingFloatChannel, default=0}
       thumb_03_l_ctrl.Scale.X {type=MovieSceneScriptingFloatChannel, default=1}
       thumb_03_l_ctrl.Scale.Y {type=MovieSceneScriptingFloatChannel, default=1}
       thumb_03_l_ctrl.Scale.Z {type=MovieSceneScriptingFloatChannel, default=1}
       index_metacarpal_l_ctrl.Location.X {type=MovieSceneScriptingFloatChannel, default=0}
       index_metacarpal_l_ctrl.Location.Y {type=MovieSceneScriptingFloatChannel, default=0}
       index_metacarpal_l_ctrl.Location.Z {type=MovieSceneScriptingFloatChannel, default=0}
       index_metacarpal_l_ctrl.Rotation.X {type=MovieSceneScriptingFloatChannel, default=0}
       index_metacarpal_l_ctrl.Rotation.Y {type=MovieSceneScriptingFloatChannel, default=0}
       index_metacarpal_l_ctrl.Rotation.Z {type=MovieSceneScriptingFloatChannel, default=0}
       index_metacarpal_l_ctrl.Scale.X {type=MovieSceneScriptingFloatChannel, default=1}
       index_metacarpal_l_ctrl.Scale.Y {type=MovieSceneScriptingFloatChannel, default=1}
       index_metacarpal_l_ctrl.Scale.Z {type=MovieSceneScriptingFloatChannel, default=1}
       index_01_l_ctrl.Location.X {type=MovieSceneScriptingFloatChannel, default=0}
       index_01_l_ctrl.Location.Y {type=MovieSceneScriptingFloatChannel, default=0}
       index_01_l_ctrl.Location.Z {type=MovieSceneScriptingFloatChannel, default=0}
       index_01_l_ctrl.Rotation.X {type=MovieSceneScriptingFloatChannel, default=0}
       index_01_l_ctrl.Rotation.Y {type=MovieSceneScriptingFloatChannel, default=0}
       index_01_l_ctrl.Rotation.Z {type=MovieSceneScriptingFloatChannel, default=0}
       index_01_l_ctrl.Scale.X {type=MovieSceneScriptingFloatChannel, default=1}
       index_01_l_ctrl.Scale.Y {type=MovieSceneScriptingFloatChannel, default=1}
       index_01_l_ctrl.Scale.Z {type=MovieSceneScriptingFloatChannel, default=1}
       index_02_l_ctrl.Location.X {type=MovieSceneScriptingFloatChannel, default=0}
       index_02_l_ctrl.Location.Y {type=MovieSceneScriptingFloatChannel, default=0}
       index_02_l_ctrl.Location.Z {type=MovieSceneScriptingFloatChannel, default=0}
       index_02_l_ctrl.Rotation.X {type=MovieSceneScriptingFloatChannel, default=0}
       index_02_l_ctrl.Rotation.Y {type=MovieSceneScriptingFloatChannel, default=0}
       index_02_l_ctrl.Rotation.Z {type=MovieSceneScriptingFloatChannel, default=0}
       index_02_l_ctrl.Scale.X {type=MovieSceneScriptingFloatChannel, default=1}
       index_02_l_ctrl.Scale.Y {type=MovieSceneScriptingFloatChannel, default=1}
       index_02_l_ctrl.Scale.Z {type=MovieSceneScriptingFloatChannel, default=1}
       index_03_l_ctrl.Location.X {type=MovieSceneScriptingFloatChannel, default=0}
       index_03_l_ctrl.Location.Y {type=MovieSceneScriptingFloatChannel, default=0}
       index_03_l_ctrl.Location.Z {type=MovieSceneScriptingFloatChannel, default=0}
       index_03_l_ctrl.Rotation.X {type=MovieSceneScriptingFloatChannel, default=0}
       index_03_l_ctrl.Rotation.Y {type=MovieSceneScriptingFloatChannel, default=0}
       index_03_l_ctrl.Rotation.Z {type=MovieSceneScriptingFloatChannel, default=0}
       index_03_l_ctrl.Scale.X {type=MovieSceneScriptingFloatChannel, default=1}
       index_03_l_ctrl.Scale.Y {type=MovieSceneScriptingFloatChannel, default=1}
       index_03_l_ctrl.Scale.Z {type=MovieSceneScriptingFloatChannel, default=1}
       middle_metacarpal_l_ctrl.Location.X {type=MovieSceneScriptingFloatChannel, default=0}
       middle_metacarpal_l_ctrl.Location.Y {type=MovieSceneScriptingFloatChannel, default=0}
       middle_metacarpal_l_ctrl.Location.Z {type=MovieSceneScriptingFloatChannel, default=0}
       middle_metacarpal_l_ctrl.Rotation.X {type=MovieSceneScriptingFloatChannel, default=0}
       middle_metacarpal_l_ctrl.Rotation.Y {type=MovieSceneScriptingFloatChannel, default=0}
       middle_metacarpal_l_ctrl.Rotation.Z {type=MovieSceneScriptingFloatChannel, default=0}
       middle_metacarpal_l_ctrl.Scale.X {type=MovieSceneScriptingFloatChannel, default=1}
       middle_metacarpal_l_ctrl.Scale.Y {type=MovieSceneScriptingFloatChannel, default=1}
       middle_metacarpal_l_ctrl.Scale.Z {type=MovieSceneScriptingFloatChannel, default=1}
       middle_01_l_ctrl.Location.X {type=MovieSceneScriptingFloatChannel, default=0}
       middle_01_l_ctrl.Location.Y {type=MovieSceneScriptingFloatChannel, default=0}
       middle_01_l_ctrl.Location.Z {type=MovieSceneScriptingFloatChannel, default=0}
       middle_01_l_ctrl.Rotation.X {type=MovieSceneScriptingFloatChannel, default=0}
       middle_01_l_ctrl.Rotation.Y {type=MovieSceneScriptingFloatChannel, default=0}
       middle_01_l_ctrl.Rotation.Z {type=MovieSceneScriptingFloatChannel, default=0}
       middle_01_l_ctrl.Scale.X {type=MovieSceneScriptingFloatChannel, default=1}
       middle_01_l_ctrl.Scale.Y {type=MovieSceneScriptingFloatChannel, default=1}
       middle_01_l_ctrl.Scale.Z {type=MovieSceneScriptingFloatChannel, default=1}
       middle_02_l_ctrl.Location.X {type=MovieSceneScriptingFloatChannel, default=0}
       middle_02_l_ctrl.Location.Y {type=MovieSceneScriptingFloatChannel, default=0}
       middle_02_l_ctrl.Location.Z {type=MovieSceneScriptingFloatChannel, default=0}
       middle_02_l_ctrl.Rotation.X {type=MovieSceneScriptingFloatChannel, default=0}
       middle_02_l_ctrl.Rotation.Y {type=MovieSceneScriptingFloatChannel, default=0}
       middle_02_l_ctrl.Rotation.Z {type=MovieSceneScriptingFloatChannel, default=0}
       middle_02_l_ctrl.Scale.X {type=MovieSceneScriptingFloatChannel, default=1}
       middle_02_l_ctrl.Scale.Y {type=MovieSceneScriptingFloatChannel, default=1}
       middle_02_l_ctrl.Scale.Z {type=MovieSceneScriptingFloatChannel, default=1}
       middle_03_l_ctrl.Location.X {type=MovieSceneScriptingFloatChannel, default=0}
       middle_03_l_ctrl.Location.Y {type=MovieSceneScriptingFloatChannel, default=0}
       middle_03_l_ctrl.Location.Z {type=MovieSceneScriptingFloatChannel, default=0}
       middle_03_l_ctrl.Rotation.X {type=MovieSceneScriptingFloatChannel, default=0}
       middle_03_l_ctrl.Rotation.Y {type=MovieSceneScriptingFloatChannel, default=0}
       middle_03_l_ctrl.Rotation.Z {type=MovieSceneScriptingFloatChannel, default=0}
       middle_03_l_ctrl.Scale.X {type=MovieSceneScriptingFloatChannel, default=1}
       middle_03_l_ctrl.Scale.Y {type=MovieSceneScriptingFloatChannel, default=1}
       middle_03_l_ctrl.Scale.Z {type=MovieSceneScriptingFloatChannel, default=1}
       ring_metacarpal_l_ctrl.Location.X {type=MovieSceneScriptingFloatChannel, default=0}
       ring_metacarpal_l_ctrl.Location.Y {type=MovieSceneScriptingFloatChannel, default=0}
       ring_metacarpal_l_ctrl.Location.Z {type=MovieSceneScriptingFloatChannel, default=0}
       ring_metacarpal_l_ctrl.Rotation.X {type=MovieSceneScriptingFloatChannel, default=0}
       ring_metacarpal_l_ctrl.Rotation.Y {type=MovieSceneScriptingFloatChannel, default=0}
       ring_metacarpal_l_ctrl.Rotation.Z {type=MovieSceneScriptingFloatChannel, default=0}
       ring_metacarpal_l_ctrl.Scale.X {type=MovieSceneScriptingFloatChannel, default=1}
       ring_metacarpal_l_ctrl.Scale.Y {type=MovieSceneScriptingFloatChannel, default=1}
       ring_metacarpal_l_ctrl.Scale.Z {type=MovieSceneScriptingFloatChannel, default=1}
       ring_01_l_ctrl.Location.X {type=MovieSceneScriptingFloatChannel, default=0}
       ring_01_l_ctrl.Location.Y {type=MovieSceneScriptingFloatChannel, default=0}
       ring_01_l_ctrl.Location.Z {type=MovieSceneScriptingFloatChannel, default=0}
       ring_01_l_ctrl.Rotation.X {type=MovieSceneScriptingFloatChannel, default=0}
       ring_01_l_ctrl.Rotation.Y {type=MovieSceneScriptingFloatChannel, default=0}
       ring_01_l_ctrl.Rotation.Z {type=MovieSceneScriptingFloatChannel, default=0}
       ring_01_l_ctrl.Scale.X {type=MovieSceneScriptingFloatChannel, default=1}
       ring_01_l_ctrl.Scale.Y {type=MovieSceneScriptingFloatChannel, default=1}
       ring_01_l_ctrl.Scale.Z {type=MovieSceneScriptingFloatChannel, default=1}
       ring_02_l_ctrl.Location.X {type=MovieSceneScriptingFloatChannel, default=0}
       ring_02_l_ctrl.Location.Y {type=MovieSceneScriptingFloatChannel, default=0}
       ring_02_l_ctrl.Location.Z {type=MovieSceneScriptingFloatChannel, default=0}
       ring_02_l_ctrl.Rotation.X {type=MovieSceneScriptingFloatChannel, default=0}
       ring_02_l_ctrl.Rotation.Y {type=MovieSceneScriptingFloatChannel, default=0}
       ring_02_l_ctrl.Rotation.Z {type=MovieSceneScriptingFloatChannel, default=0}
       ring_02_l_ctrl.Scale.X {type=MovieSceneScriptingFloatChannel, default=1}
       ring_02_l_ctrl.Scale.Y {type=MovieSceneScriptingFloatChannel, default=1}
       ring_02_l_ctrl.Scale.Z {type=MovieSceneScriptingFloatChannel, default=1}
       ring_03_l_ctrl.Location.X {type=MovieSceneScriptingFloatChannel, default=0}
       ring_03_l_ctrl.Location.Y {type=MovieSceneScriptingFloatChannel, default=0}
       ring_03_l_ctrl.Location.Z {type=MovieSceneScriptingFloatChannel, default=0}
       ring_03_l_ctrl.Rotation.X {type=MovieSceneScriptingFloatChannel, default=0}
       ring_03_l_ctrl.Rotation.Y {type=MovieSceneScriptingFloatChannel, default=0}
       ring_03_l_ctrl.Rotation.Z {type=MovieSceneScriptingFloatChannel, default=0}
       ring_03_l_ctrl.Scale.X {type=MovieSceneScriptingFloatChannel, default=1}
       ring_03_l_ctrl.Scale.Y {type=MovieSceneScriptingFloatChannel, default=1}
       ring_03_l_ctrl.Scale.Z {type=MovieSceneScriptingFloatChannel, default=1}
       pinky_metacarpal_l_ctrl.Location.X {type=MovieSceneScriptingFloatChannel, default=0}
       pinky_metacarpal_l_ctrl.Location.Y {type=MovieSceneScriptingFloatChannel, default=0}
       pinky_metacarpal_l_ctrl.Location.Z {type=MovieSceneScriptingFloatChannel, default=0}
       pinky_metacarpal_l_ctrl.Rotation.X {type=MovieSceneScriptingFloatChannel, default=0}
       pinky_metacarpal_l_ctrl.Rotation.Y {type=MovieSceneScriptingFloatChannel, default=0}
       pinky_metacarpal_l_ctrl.Rotation.Z {type=MovieSceneScriptingFloatChannel, default=0}
       pinky_metacarpal_l_ctrl.Scale.X {type=MovieSceneScriptingFloatChannel, default=1}
       pinky_metacarpal_l_ctrl.Scale.Y {type=MovieSceneScriptingFloatChannel, default=1}
       pinky_metacarpal_l_ctrl.Scale.Z {type=MovieSceneScriptingFloatChannel, default=1}
       pinky_01_l_ctrl.Location.X {type=MovieSceneScriptingFloatChannel, default=0}
       pinky_01_l_ctrl.Location.Y {type=MovieSceneScriptingFloatChannel, default=0}
       pinky_01_l_ctrl.Location.Z {type=MovieSceneScriptingFloatChannel, default=0}
       pinky_01_l_ctrl.Rotation.X {type=MovieSceneScriptingFloatChannel, default=0}
       pinky_01_l_ctrl.Rotation.Y {type=MovieSceneScriptingFloatChannel, default=0}
       pinky_01_l_ctrl.Rotation.Z {type=MovieSceneScriptingFloatChannel, default=0}
       pinky_01_l_ctrl.Scale.X {type=MovieSceneScriptingFloatChannel, default=1}
       pinky_01_l_ctrl.Scale.Y {type=MovieSceneScriptingFloatChannel, default=1}
       pinky_01_l_ctrl.Scale.Z {type=MovieSceneScriptingFloatChannel, default=1}
       pinky_02_l_ctrl.Location.X {type=MovieSceneScriptingFloatChannel, default=0}
       pinky_02_l_ctrl.Location.Y {type=MovieSceneScriptingFloatChannel, default=0}
       pinky_02_l_ctrl.Location.Z {type=MovieSceneScriptingFloatChannel, default=0}
       pinky_02_l_ctrl.Rotation.X {type=MovieSceneScriptingFloatChannel, default=0}
       pinky_02_l_ctrl.Rotation.Y {type=MovieSceneScriptingFloatChannel, default=0}
       pinky_02_l_ctrl.Rotation.Z {type=MovieSceneScriptingFloatChannel, default=0}
       pinky_02_l_ctrl.Scale.X {type=MovieSceneScriptingFloatChannel, default=1}
       pinky_02_l_ctrl.Scale.Y {type=MovieSceneScriptingFloatChannel, default=1}
       pinky_02_l_ctrl.Scale.Z {type=MovieSceneScriptingFloatChannel, default=1}
       pinky_03_l_ctrl.Location.X {type=MovieSceneScriptingFloatChannel, default=0}
       pinky_03_l_ctrl.Location.Y {type=MovieSceneScriptingFloatChannel, default=0}
       pinky_03_l_ctrl.Location.Z {type=MovieSceneScriptingFloatChannel, default=0}
       pinky_03_l_ctrl.Rotation.X {type=MovieSceneScriptingFloatChannel, default=0}
       pinky_03_l_ctrl.Rotation.Y {type=MovieSceneScriptingFloatChannel, default=0}
       pinky_03_l_ctrl.Rotation.Z {type=MovieSceneScriptingFloatChannel, default=0}
       pinky_03_l_ctrl.Scale.X {type=MovieSceneScriptingFloatChannel, default=1}
       pinky_03_l_ctrl.Scale.Y {type=MovieSceneScriptingFloatChannel, default=1}
       pinky_03_l_ctrl.Scale.Z {type=MovieSceneScriptingFloatChannel, default=1}
       hand_l_ik_ctrl.Location.X {type=MovieSceneScriptingFloatChannel, default=0}
       hand_l_ik_ctrl.Location.Y {type=MovieSceneScriptingFloatChannel, default=0}
       hand_l_ik_ctrl.Location.Z {type=MovieSceneScriptingFloatChannel, default=0}
       hand_l_ik_ctrl.Rotation.X {type=MovieSceneScriptingFloatChannel, default=0}
       hand_l_ik_ctrl.Rotation.Y {type=MovieSceneScriptingFloatChannel, default=0}
       hand_l_ik_ctrl.Rotation.Z {type=MovieSceneScriptingFloatChannel, default=0}
       hand_l_ik_ctrl.Scale.X {type=MovieSceneScriptingFloatChannel, default=1}
       hand_l_ik_ctrl.Scale.Y {type=MovieSceneScriptingFloatChannel, default=1}
       hand_l_ik_ctrl.Scale.Z {type=MovieSceneScriptingFloatChannel, default=1}
       Softness {type=MovieSceneScriptingFloatChannel, default=0}
       hand_l_ikorient_ctrl.Rotation.Roll {type=MovieSceneScriptingFloatChannel, default=0}
       hand_l_ikorient_ctrl.Rotation.Pitch {type=MovieSceneScriptingFloatChannel, default=0}
       hand_l_ikorient_ctrl.Rotation.Yaw {type=MovieSceneScriptingFloatChannel, default=0}
       arm_l_pv_ik_ctrl.Location.X {type=MovieSceneScriptingFloatChannel, default=0}
       arm_l_pv_ik_ctrl.Location.Y {type=MovieSceneScriptingFloatChannel, default=0}
       arm_l_pv_ik_ctrl.Location.Z {type=MovieSceneScriptingFloatChannel, default=0}
       arm_l_pv_ik_ctrl.Rotation.X {type=MovieSceneScriptingFloatChannel, default=0}
       arm_l_pv_ik_ctrl.Rotation.Y {type=MovieSceneScriptingFloatChannel, default=0}
       arm_l_pv_ik_ctrl.Rotation.Z {type=MovieSceneScriptingFloatChannel, default=0}
       arm_l_pv_ik_ctrl.Scale.X {type=MovieSceneScriptingFloatChannel, default=1}
       arm_l_pv_ik_ctrl.Scale.Y {type=MovieSceneScriptingFloatChannel, default=1}
       arm_l_pv_ik_ctrl.Scale.Z {type=MovieSceneScriptingFloatChannel, default=1}
       upperarm_r_fk_ctrl.Location.X {type=MovieSceneScriptingFloatChannel, default=0}
       upperarm_r_fk_ctrl.Location.Y {type=MovieSceneScriptingFloatChannel, default=0}
       upperarm_r_fk_ctrl.Location.Z {type=MovieSceneScriptingFloatChannel, default=0}
       upperarm_r_fk_ctrl.Rotation.X {type=MovieSceneScriptingFloatChannel, default=0}
       upperarm_r_fk_ctrl.Rotation.Y {type=MovieSceneScriptingFloatChannel, default=0}
       upperarm_r_fk_ctrl.Rotation.Z {type=MovieSceneScriptingFloatChannel, default=0}
       upperarm_r_fk_ctrl.Scale.X {type=MovieSceneScriptingFloatChannel, default=1}
       upperarm_r_fk_ctrl.Scale.Y {type=MovieSceneScriptingFloatChannel, default=1}
       upperarm_r_fk_ctrl.Scale.Z {type=MovieSceneScriptingFloatChannel, default=1}
       lowerarm_r_fk_ctrl.Location.X {type=MovieSceneScriptingFloatChannel, default=0}
       lowerarm_r_fk_ctrl.Location.Y {type=MovieSceneScriptingFloatChannel, default=0}
       lowerarm_r_fk_ctrl.Location.Z {type=MovieSceneScriptingFloatChannel, default=0}
       lowerarm_r_fk_ctrl.Rotation.X {type=MovieSceneScriptingFloatChannel, default=0}
       lowerarm_r_fk_ctrl.Rotation.Y {type=MovieSceneScriptingFloatChannel, default=0}
       lowerarm_r_fk_ctrl.Rotation.Z {type=MovieSceneScriptingFloatChannel, default=0}
       lowerarm_r_fk_ctrl.Scale.X {type=MovieSceneScriptingFloatChannel, default=1}
       lowerarm_r_fk_ctrl.Scale.Y {type=MovieSceneScriptingFloatChannel, default=1}
       lowerarm_r_fk_ctrl.Scale.Z {type=MovieSceneScriptingFloatChannel, default=1}
       hand_r_fk_ctrl.Location.X {type=MovieSceneScriptingFloatChannel, default=0}
       hand_r_fk_ctrl.Location.Y {type=MovieSceneScriptingFloatChannel, default=0}
       hand_r_fk_ctrl.Location.Z {type=MovieSceneScriptingFloatChannel, default=0}
       hand_r_fk_ctrl.Rotation.X {type=MovieSceneScriptingFloatChannel, default=0}
       hand_r_fk_ctrl.Rotation.Y {type=MovieSceneScriptingFloatChannel, default=0}
       hand_r_fk_ctrl.Rotation.Z {type=MovieSceneScriptingFloatChannel, default=0}
       hand_r_fk_ctrl.Scale.X {type=MovieSceneScriptingFloatChannel, default=1}
       hand_r_fk_ctrl.Scale.Y {type=MovieSceneScriptingFloatChannel, default=1}
       hand_r_fk_ctrl.Scale.Z {type=MovieSceneScriptingFloatChannel, default=1}
       thumb_01_r_ctrl.Location.X {type=MovieSceneScriptingFloatChannel, default=0}
       thumb_01_r_ctrl.Location.Y {type=MovieSceneScriptingFloatChannel, default=0}
       thumb_01_r_ctrl.Location.Z {type=MovieSceneScriptingFloatChannel, default=0}
       thumb_01_r_ctrl.Rotation.X {type=MovieSceneScriptingFloatChannel, default=0}
       thumb_01_r_ctrl.Rotation.Y {type=MovieSceneScriptingFloatChannel, default=0}
       thumb_01_r_ctrl.Rotation.Z {type=MovieSceneScriptingFloatChannel, default=0}
       thumb_01_r_ctrl.Scale.X {type=MovieSceneScriptingFloatChannel, default=1}
       thumb_01_r_ctrl.Scale.Y {type=MovieSceneScriptingFloatChannel, default=1}
       thumb_01_r_ctrl.Scale.Z {type=MovieSceneScriptingFloatChannel, default=1}
       thumb_02_r_ctrl.Location.X {type=MovieSceneScriptingFloatChannel, default=0}
       thumb_02_r_ctrl.Location.Y {type=MovieSceneScriptingFloatChannel, default=0}
       thumb_02_r_ctrl.Location.Z {type=MovieSceneScriptingFloatChannel, default=0}
       thumb_02_r_ctrl.Rotation.X {type=MovieSceneScriptingFloatChannel, default=0}
       thumb_02_r_ctrl.Rotation.Y {type=MovieSceneScriptingFloatChannel, default=0}
       thumb_02_r_ctrl.Rotation.Z {type=MovieSceneScriptingFloatChannel, default=0}
       thumb_02_r_ctrl.Scale.X {type=MovieSceneScriptingFloatChannel, default=1}
       thumb_02_r_ctrl.Scale.Y {type=MovieSceneScriptingFloatChannel, default=1}
       thumb_02_r_ctrl.Scale.Z {type=MovieSceneScriptingFloatChannel, default=1}
       thumb_03_r_ctrl.Location.X {type=MovieSceneScriptingFloatChannel, default=0}
       thumb_03_r_ctrl.Location.Y {type=MovieSceneScriptingFloatChannel, default=0}
       thumb_03_r_ctrl.Location.Z {type=MovieSceneScriptingFloatChannel, default=0}
       thumb_03_r_ctrl.Rotation.X {type=MovieSceneScriptingFloatChannel, default=0}
       thumb_03_r_ctrl.Rotation.Y {type=MovieSceneScriptingFloatChannel, default=0}
       thumb_03_r_ctrl.Rotation.Z {type=MovieSceneScriptingFloatChannel, default=0}
       thumb_03_r_ctrl.Scale.X {type=MovieSceneScriptingFloatChannel, default=1}
       thumb_03_r_ctrl.Scale.Y {type=MovieSceneScriptingFloatChannel, default=1}
       thumb_03_r_ctrl.Scale.Z {type=MovieSceneScriptingFloatChannel, default=1}
       index_metacarpal_r_ctrl.Location.X {type=MovieSceneScriptingFloatChannel, default=0}
       index_metacarpal_r_ctrl.Location.Y {type=MovieSceneScriptingFloatChannel, default=0}
       index_metacarpal_r_ctrl.Location.Z {type=MovieSceneScriptingFloatChannel, default=0}
       index_metacarpal_r_ctrl.Rotation.X {type=MovieSceneScriptingFloatChannel, default=0}
       index_metacarpal_r_ctrl.Rotation.Y {type=MovieSceneScriptingFloatChannel, default=0}
       index_metacarpal_r_ctrl.Rotation.Z {type=MovieSceneScriptingFloatChannel, default=0}
       index_metacarpal_r_ctrl.Scale.X {type=MovieSceneScriptingFloatChannel, default=1}
       index_metacarpal_r_ctrl.Scale.Y {type=MovieSceneScriptingFloatChannel, default=1}
       index_metacarpal_r_ctrl.Scale.Z {type=MovieSceneScriptingFloatChannel, default=1}
       index_01_r_ctrl.Location.X {type=MovieSceneScriptingFloatChannel, default=0}
       index_01_r_ctrl.Location.Y {type=MovieSceneScriptingFloatChannel, default=0}
       index_01_r_ctrl.Location.Z {type=MovieSceneScriptingFloatChannel, default=0}
       index_01_r_ctrl.Rotation.X {type=MovieSceneScriptingFloatChannel, default=0}
       index_01_r_ctrl.Rotation.Y {type=MovieSceneScriptingFloatChannel, default=0}
       index_01_r_ctrl.Rotation.Z {type=MovieSceneScriptingFloatChannel, default=0}
       index_01_r_ctrl.Scale.X {type=MovieSceneScriptingFloatChannel, default=1}
       index_01_r_ctrl.Scale.Y {type=MovieSceneScriptingFloatChannel, default=1}
       index_01_r_ctrl.Scale.Z {type=MovieSceneScriptingFloatChannel, default=1}
       index_02_r_ctrl.Location.X {type=MovieSceneScriptingFloatChannel, default=0}
       index_02_r_ctrl.Location.Y {type=MovieSceneScriptingFloatChannel, default=0}
       index_02_r_ctrl.Location.Z {type=MovieSceneScriptingFloatChannel, default=0}
       index_02_r_ctrl.Rotation.X {type=MovieSceneScriptingFloatChannel, default=0}
       index_02_r_ctrl.Rotation.Y {type=MovieSceneScriptingFloatChannel, default=0}
       index_02_r_ctrl.Rotation.Z {type=MovieSceneScriptingFloatChannel, default=0}
       index_02_r_ctrl.Scale.X {type=MovieSceneScriptingFloatChannel, default=1}
       index_02_r_ctrl.Scale.Y {type=MovieSceneScriptingFloatChannel, default=1}
       index_02_r_ctrl.Scale.Z {type=MovieSceneScriptingFloatChannel, default=1}
       index_03_r_ctrl.Location.X {type=MovieSceneScriptingFloatChannel, default=0}
       index_03_r_ctrl.Location.Y {type=MovieSceneScriptingFloatChannel, default=0}
       index_03_r_ctrl.Location.Z {type=MovieSceneScriptingFloatChannel, default=0}
       index_03_r_ctrl.Rotation.X {type=MovieSceneScriptingFloatChannel, default=0}
       index_03_r_ctrl.Rotation.Y {type=MovieSceneScriptingFloatChannel, default=0}
       index_03_r_ctrl.Rotation.Z {type=MovieSceneScriptingFloatChannel, default=0}
       index_03_r_ctrl.Scale.X {type=MovieSceneScriptingFloatChannel, default=1}
       index_03_r_ctrl.Scale.Y {type=MovieSceneScriptingFloatChannel, default=1}
       index_03_r_ctrl.Scale.Z {type=MovieSceneScriptingFloatChannel, default=1}
       middle_metacarpal_r_ctrl.Location.X {type=MovieSceneScriptingFloatChannel, default=0}
       middle_metacarpal_r_ctrl.Location.Y {type=MovieSceneScriptingFloatChannel, default=0}
       middle_metacarpal_r_ctrl.Location.Z {type=MovieSceneScriptingFloatChannel, default=0}
       middle_metacarpal_r_ctrl.Rotation.X {type=MovieSceneScriptingFloatChannel, default=0}
       middle_metacarpal_r_ctrl.Rotation.Y {type=MovieSceneScriptingFloatChannel, default=0}
       middle_metacarpal_r_ctrl.Rotation.Z {type=MovieSceneScriptingFloatChannel, default=0}
       middle_metacarpal_r_ctrl.Scale.X {type=MovieSceneScriptingFloatChannel, default=1}
       middle_metacarpal_r_ctrl.Scale.Y {type=MovieSceneScriptingFloatChannel, default=1}
       middle_metacarpal_r_ctrl.Scale.Z {type=MovieSceneScriptingFloatChannel, default=1}
       middle_01_r_ctrl.Location.X {type=MovieSceneScriptingFloatChannel, default=0}
       middle_01_r_ctrl.Location.Y {type=MovieSceneScriptingFloatChannel, default=0}
       middle_01_r_ctrl.Location.Z {type=MovieSceneScriptingFloatChannel, default=0}
       middle_01_r_ctrl.Rotation.X {type=MovieSceneScriptingFloatChannel, default=0}
       middle_01_r_ctrl.Rotation.Y {type=MovieSceneScriptingFloatChannel, default=0}
       middle_01_r_ctrl.Rotation.Z {type=MovieSceneScriptingFloatChannel, default=0}
       middle_01_r_ctrl.Scale.X {type=MovieSceneScriptingFloatChannel, default=1}
       middle_01_r_ctrl.Scale.Y {type=MovieSceneScriptingFloatChannel, default=1}
       middle_01_r_ctrl.Scale.Z {type=MovieSceneScriptingFloatChannel, default=1}
       middle_02_r_ctrl.Location.X {type=MovieSceneScriptingFloatChannel, default=0}
       middle_02_r_ctrl.Location.Y {type=MovieSceneScriptingFloatChannel, default=0}
       middle_02_r_ctrl.Location.Z {type=MovieSceneScriptingFloatChannel, default=0}
       middle_02_r_ctrl.Rotation.X {type=MovieSceneScriptingFloatChannel, default=0}
       middle_02_r_ctrl.Rotation.Y {type=MovieSceneScriptingFloatChannel, default=0}
       middle_02_r_ctrl.Rotation.Z {type=MovieSceneScriptingFloatChannel, default=0}
       middle_02_r_ctrl.Scale.X {type=MovieSceneScriptingFloatChannel, default=1}
       middle_02_r_ctrl.Scale.Y {type=MovieSceneScriptingFloatChannel, default=1}
       middle_02_r_ctrl.Scale.Z {type=MovieSceneScriptingFloatChannel, default=1}
       middle_03_r_ctrl.Location.X {type=MovieSceneScriptingFloatChannel, default=0}
       middle_03_r_ctrl.Location.Y {type=MovieSceneScriptingFloatChannel, default=0}
       middle_03_r_ctrl.Location.Z {type=MovieSceneScriptingFloatChannel, default=0}
       middle_03_r_ctrl.Rotation.X {type=MovieSceneScriptingFloatChannel, default=0}
       middle_03_r_ctrl.Rotation.Y {type=MovieSceneScriptingFloatChannel, default=0}
       middle_03_r_ctrl.Rotation.Z {type=MovieSceneScriptingFloatChannel, default=0}
       middle_03_r_ctrl.Scale.X {type=MovieSceneScriptingFloatChannel, default=1}
       middle_03_r_ctrl.Scale.Y {type=MovieSceneScriptingFloatChannel, default=1}
       middle_03_r_ctrl.Scale.Z {type=MovieSceneScriptingFloatChannel, default=1}
       ring_metacarpal_r_ctrl.Location.X {type=MovieSceneScriptingFloatChannel, default=0}
       ring_metacarpal_r_ctrl.Location.Y {type=MovieSceneScriptingFloatChannel, default=0}
       ring_metacarpal_r_ctrl.Location.Z {type=MovieSceneScriptingFloatChannel, default=0}
       ring_metacarpal_r_ctrl.Rotation.X {type=MovieSceneScriptingFloatChannel, default=0}
       ring_metacarpal_r_ctrl.Rotation.Y {type=MovieSceneScriptingFloatChannel, default=0}
       ring_metacarpal_r_ctrl.Rotation.Z {type=MovieSceneScriptingFloatChannel, default=0}
       ring_metacarpal_r_ctrl.Scale.X {type=MovieSceneScriptingFloatChannel, default=1}
       ring_metacarpal_r_ctrl.Scale.Y {type=MovieSceneScriptingFloatChannel, default=1}
       ring_metacarpal_r_ctrl.Scale.Z {type=MovieSceneScriptingFloatChannel, default=1}
       ring_01_r_ctrl.Location.X {type=MovieSceneScriptingFloatChannel, default=0}
       ring_01_r_ctrl.Location.Y {type=MovieSceneScriptingFloatChannel, default=0}
       ring_01_r_ctrl.Location.Z {type=MovieSceneScriptingFloatChannel, default=0}
       ring_01_r_ctrl.Rotation.X {type=MovieSceneScriptingFloatChannel, default=0}
       ring_01_r_ctrl.Rotation.Y {type=MovieSceneScriptingFloatChannel, default=0}
       ring_01_r_ctrl.Rotation.Z {type=MovieSceneScriptingFloatChannel, default=0}
       ring_01_r_ctrl.Scale.X {type=MovieSceneScriptingFloatChannel, default=1}
       ring_01_r_ctrl.Scale.Y {type=MovieSceneScriptingFloatChannel, default=1}
       ring_01_r_ctrl.Scale.Z {type=MovieSceneScriptingFloatChannel, default=1}
       ring_02_r_ctrl.Location.X {type=MovieSceneScriptingFloatChannel, default=0}
       ring_02_r_ctrl.Location.Y {type=MovieSceneScriptingFloatChannel, default=0}
       ring_02_r_ctrl.Location.Z {type=MovieSceneScriptingFloatChannel, default=0}
       ring_02_r_ctrl.Rotation.X {type=MovieSceneScriptingFloatChannel, default=0}
       ring_02_r_ctrl.Rotation.Y {type=MovieSceneScriptingFloatChannel, default=0}
       ring_02_r_ctrl.Rotation.Z {type=MovieSceneScriptingFloatChannel, default=0}
       ring_02_r_ctrl.Scale.X {type=MovieSceneScriptingFloatChannel, default=1}
       ring_02_r_ctrl.Scale.Y {type=MovieSceneScriptingFloatChannel, default=1}
       ring_02_r_ctrl.Scale.Z {type=MovieSceneScriptingFloatChannel, default=1}
       ring_03_r_ctrl.Location.X {type=MovieSceneScriptingFloatChannel, default=0}
       ring_03_r_ctrl.Location.Y {type=MovieSceneScriptingFloatChannel, default=0}
       ring_03_r_ctrl.Location.Z {type=MovieSceneScriptingFloatChannel, default=0}
       ring_03_r_ctrl.Rotation.X {type=MovieSceneScriptingFloatChannel, default=0}
       ring_03_r_ctrl.Rotation.Y {type=MovieSceneScriptingFloatChannel, default=0}
       ring_03_r_ctrl.Rotation.Z {type=MovieSceneScriptingFloatChannel, default=0}
       ring_03_r_ctrl.Scale.X {type=MovieSceneScriptingFloatChannel, default=1}
       ring_03_r_ctrl.Scale.Y {type=MovieSceneScriptingFloatChannel, default=1}
       ring_03_r_ctrl.Scale.Z {type=MovieSceneScriptingFloatChannel, default=1}
       pinky_metacarpal_r_ctrl.Location.X {type=MovieSceneScriptingFloatChannel, default=0}
       pinky_metacarpal_r_ctrl.Location.Y {type=MovieSceneScriptingFloatChannel, default=0}
       pinky_metacarpal_r_ctrl.Location.Z {type=MovieSceneScriptingFloatChannel, default=0}
       pinky_metacarpal_r_ctrl.Rotation.X {type=MovieSceneScriptingFloatChannel, default=0}
       pinky_metacarpal_r_ctrl.Rotation.Y {type=MovieSceneScriptingFloatChannel, default=0}
       pinky_metacarpal_r_ctrl.Rotation.Z {type=MovieSceneScriptingFloatChannel, default=0}
       pinky_metacarpal_r_ctrl.Scale.X {type=MovieSceneScriptingFloatChannel, default=1}
       pinky_metacarpal_r_ctrl.Scale.Y {type=MovieSceneScriptingFloatChannel, default=1}
       pinky_metacarpal_r_ctrl.Scale.Z {type=MovieSceneScriptingFloatChannel, default=1}
       pinky_01_r_ctrl.Location.X {type=MovieSceneScriptingFloatChannel, default=0}
       pinky_01_r_ctrl.Location.Y {type=MovieSceneScriptingFloatChannel, default=0}
       pinky_01_r_ctrl.Location.Z {type=MovieSceneScriptingFloatChannel, default=0}
       pinky_01_r_ctrl.Rotation.X {type=MovieSceneScriptingFloatChannel, default=0}
       pinky_01_r_ctrl.Rotation.Y {type=MovieSceneScriptingFloatChannel, default=0}
       pinky_01_r_ctrl.Rotation.Z {type=MovieSceneScriptingFloatChannel, default=0}
       pinky_01_r_ctrl.Scale.X {type=MovieSceneScriptingFloatChannel, default=1}
       pinky_01_r_ctrl.Scale.Y {type=MovieSceneScriptingFloatChannel, default=1}
       pinky_01_r_ctrl.Scale.Z {type=MovieSceneScriptingFloatChannel, default=1}
       pinky_02_r_ctrl.Location.X {type=MovieSceneScriptingFloatChannel, default=0}
       pinky_02_r_ctrl.Location.Y {type=MovieSceneScriptingFloatChannel, default=0}
       pinky_02_r_ctrl.Location.Z {type=MovieSceneScriptingFloatChannel, default=0}
       pinky_02_r_ctrl.Rotation.X {type=MovieSceneScriptingFloatChannel, default=0}
       pinky_02_r_ctrl.Rotation.Y {type=MovieSceneScriptingFloatChannel, default=0}
       pinky_02_r_ctrl.Rotation.Z {type=MovieSceneScriptingFloatChannel, default=0}
       pinky_02_r_ctrl.Scale.X {type=MovieSceneScriptingFloatChannel, default=1}
       pinky_02_r_ctrl.Scale.Y {type=MovieSceneScriptingFloatChannel, default=1}
       pinky_02_r_ctrl.Scale.Z {type=MovieSceneScriptingFloatChannel, default=1}
       pinky_03_r_ctrl.Location.X {type=MovieSceneScriptingFloatChannel, default=0}
       pinky_03_r_ctrl.Location.Y {type=MovieSceneScriptingFloatChannel, default=0}
       pinky_03_r_ctrl.Location.Z {type=MovieSceneScriptingFloatChannel, default=0}
       pinky_03_r_ctrl.Rotation.X {type=MovieSceneScriptingFloatChannel, default=0}
       pinky_03_r_ctrl.Rotation.Y {type=MovieSceneScriptingFloatChannel, default=0}
       pinky_03_r_ctrl.Rotation.Z {type=MovieSceneScriptingFloatChannel, default=0}
       pinky_03_r_ctrl.Scale.X {type=MovieSceneScriptingFloatChannel, default=1}
       pinky_03_r_ctrl.Scale.Y {type=MovieSceneScriptingFloatChannel, default=1}
       pinky_03_r_ctrl.Scale.Z {type=MovieSceneScriptingFloatChannel, default=1}
       hand_r_ik_ctrl.Location.X {type=MovieSceneScriptingFloatChannel, default=0}
       hand_r_ik_ctrl.Location.Y {type=MovieSceneScriptingFloatChannel, default=0}
       hand_r_ik_ctrl.Location.Z {type=MovieSceneScriptingFloatChannel, default=0}
       hand_r_ik_ctrl.Rotation.X {type=MovieSceneScriptingFloatChannel, default=0}
       hand_r_ik_ctrl.Rotation.Y {type=MovieSceneScriptingFloatChannel, default=0}
       hand_r_ik_ctrl.Rotation.Z {type=MovieSceneScriptingFloatChannel, default=0}
       hand_r_ik_ctrl.Scale.X {type=MovieSceneScriptingFloatChannel, default=1}
       hand_r_ik_ctrl.Scale.Y {type=MovieSceneScriptingFloatChannel, default=1}
       hand_r_ik_ctrl.Scale.Z {type=MovieSceneScriptingFloatChannel, default=1}
       Softness_2 {type=MovieSceneScriptingFloatChannel, default=0}
       hand_r_ikorient_ctrl.Rotation.Roll {type=MovieSceneScriptingFloatChannel, default=0}
       hand_r_ikorient_ctrl.Rotation.Pitch {type=MovieSceneScriptingFloatChannel, default=0}
       hand_r_ikorient_ctrl.Rotation.Yaw {type=MovieSceneScriptingFloatChannel, default=0}
       arm_r_pv_ik_ctrl.Location.X {type=MovieSceneScriptingFloatChannel, default=0}
       arm_r_pv_ik_ctrl.Location.Y {type=MovieSceneScriptingFloatChannel, default=0}
       arm_r_pv_ik_ctrl.Location.Z {type=MovieSceneScriptingFloatChannel, default=0}
       arm_r_pv_ik_ctrl.Rotation.X {type=MovieSceneScriptingFloatChannel, default=0}
       arm_r_pv_ik_ctrl.Rotation.Y {type=MovieSceneScriptingFloatChannel, default=0}
       arm_r_pv_ik_ctrl.Rotation.Z {type=MovieSceneScriptingFloatChannel, default=0}
       arm_r_pv_ik_ctrl.Scale.X {type=MovieSceneScriptingFloatChannel, default=1}
       arm_r_pv_ik_ctrl.Scale.Y {type=MovieSceneScriptingFloatChannel, default=1}
       arm_r_pv_ik_ctrl.Scale.Z {type=MovieSceneScriptingFloatChannel, default=1}
       body_ctrl.Location.X {type=MovieSceneScriptingFloatChannel, default=0}
       body_ctrl.Location.Y {type=MovieSceneScriptingFloatChannel, default=0}
       body_ctrl.Location.Z {type=MovieSceneScriptingFloatChannel, default=0}
       body_ctrl.Rotation.X {type=MovieSceneScriptingFloatChannel, default=0}
       body_ctrl.Rotation.Y {type=MovieSceneScriptingFloatChannel, default=0}
       body_ctrl.Rotation.Z {type=MovieSceneScriptingFloatChannel, default=0}
       body_ctrl.Scale.X {type=MovieSceneScriptingFloatChannel, default=1}
       body_ctrl.Scale.Y {type=MovieSceneScriptingFloatChannel, default=1}
       body_ctrl.Scale.Z {type=MovieSceneScriptingFloatChannel, default=1}
       hips_ctrl.Location.X {type=MovieSceneScriptingFloatChannel, default=0}
       hips_ctrl.Location.Y {type=MovieSceneScriptingFloatChannel, default=0}
       hips_ctrl.Location.Z {type=MovieSceneScriptingFloatChannel, default=0}
       hips_ctrl.Rotation.X {type=MovieSceneScriptingFloatChannel, default=0}
       hips_ctrl.Rotation.Y {type=MovieSceneScriptingFloatChannel, default=0}
       hips_ctrl.Rotation.Z {type=MovieSceneScriptingFloatChannel, default=0}
       hips_ctrl.Scale.X {type=MovieSceneScriptingFloatChannel, default=1}
       hips_ctrl.Scale.Y {type=MovieSceneScriptingFloatChannel, default=1}
       hips_ctrl.Scale.Z {type=MovieSceneScriptingFloatChannel, default=1}
       hips_tan_ctrl.Location.X {type=MovieSceneScriptingFloatChannel, default=0}
       hips_tan_ctrl.Location.Y {type=MovieSceneScriptingFloatChannel, default=0}
       hips_tan_ctrl.Location.Z {type=MovieSceneScriptingFloatChannel, default=0}
       hips_tan_ctrl.Rotation.X {type=MovieSceneScriptingFloatChannel, default=0}
       hips_tan_ctrl.Rotation.Y {type=MovieSceneScriptingFloatChannel, default=0}
       hips_tan_ctrl.Rotation.Z {type=MovieSceneScriptingFloatChannel, default=0}
       hips_tan_ctrl.Scale.X {type=MovieSceneScriptingFloatChannel, default=1}
       hips_tan_ctrl.Scale.Y {type=MovieSceneScriptingFloatChannel, default=1}
       hips_tan_ctrl.Scale.Z {type=MovieSceneScriptingFloatChannel, default=1}
       spine_01_ctrl.Location.X {type=MovieSceneScriptingFloatChannel, default=0}
       spine_01_ctrl.Location.Y {type=MovieSceneScriptingFloatChannel, default=0}
       spine_01_ctrl.Location.Z {type=MovieSceneScriptingFloatChannel, default=0}
       spine_01_ctrl.Rotation.X {type=MovieSceneScriptingFloatChannel, default=0}
       spine_01_ctrl.Rotation.Y {type=MovieSceneScriptingFloatChannel, default=0}
       spine_01_ctrl.Rotation.Z {type=MovieSceneScriptingFloatChannel, default=0}
       spine_01_ctrl.Scale.X {type=MovieSceneScriptingFloatChannel, default=1}
       spine_01_ctrl.Scale.Y {type=MovieSceneScriptingFloatChannel, default=1}
       spine_01_ctrl.Scale.Z {type=MovieSceneScriptingFloatChannel, default=1}
       spine_02_ctrl.Location.X {type=MovieSceneScriptingFloatChannel, default=0}
       spine_02_ctrl.Location.Y {type=MovieSceneScriptingFloatChannel, default=0}
       spine_02_ctrl.Location.Z {type=MovieSceneScriptingFloatChannel, default=0}
       spine_02_ctrl.Rotation.X {type=MovieSceneScriptingFloatChannel, default=0}
       spine_02_ctrl.Rotation.Y {type=MovieSceneScriptingFloatChannel, default=0}
       spine_02_ctrl.Rotation.Z {type=MovieSceneScriptingFloatChannel, default=0}
       spine_02_ctrl.Scale.X {type=MovieSceneScriptingFloatChannel, default=1}
       spine_02_ctrl.Scale.Y {type=MovieSceneScriptingFloatChannel, default=1}
       spine_02_ctrl.Scale.Z {type=MovieSceneScriptingFloatChannel, default=1}
       spine_03_ctrl.Location.X {type=MovieSceneScriptingFloatChannel, default=0}
       spine_03_ctrl.Location.Y {type=MovieSceneScriptingFloatChannel, default=0}
       spine_03_ctrl.Location.Z {type=MovieSceneScriptingFloatChannel, default=0}
       spine_03_ctrl.Rotation.X {type=MovieSceneScriptingFloatChannel, default=0}
       spine_03_ctrl.Rotation.Y {type=MovieSceneScriptingFloatChannel, default=0}
       spine_03_ctrl.Rotation.Z {type=MovieSceneScriptingFloatChannel, default=0}
       spine_03_ctrl.Scale.X {type=MovieSceneScriptingFloatChannel, default=1}
       spine_03_ctrl.Scale.Y {type=MovieSceneScriptingFloatChannel, default=1}
       spine_03_ctrl.Scale.Z {type=MovieSceneScriptingFloatChannel, default=1}
       spine_04_ctrl.Location.X {type=MovieSceneScriptingFloatChannel, default=0}
       spine_04_ctrl.Location.Y {type=MovieSceneScriptingFloatChannel, default=0}
       spine_04_ctrl.Location.Z {type=MovieSceneScriptingFloatChannel, default=0}
       spine_04_ctrl.Rotation.X {type=MovieSceneScriptingFloatChannel, default=0}
       spine_04_ctrl.Rotation.Y {type=MovieSceneScriptingFloatChannel, default=0}
       spine_04_ctrl.Rotation.Z {type=MovieSceneScriptingFloatChannel, default=0}
       spine_04_ctrl.Scale.X {type=MovieSceneScriptingFloatChannel, default=1}
       spine_04_ctrl.Scale.Y {type=MovieSceneScriptingFloatChannel, default=1}
       spine_04_ctrl.Scale.Z {type=MovieSceneScriptingFloatChannel, default=1}
       head_ctrl.Location.X {type=MovieSceneScriptingFloatChannel, default=0}
       head_ctrl.Location.Y {type=MovieSceneScriptingFloatChannel, default=0}
       head_ctrl.Location.Z {type=MovieSceneScriptingFloatChannel, default=0}
       head_ctrl.Rotation.X {type=MovieSceneScriptingFloatChannel, default=0}
       head_ctrl.Rotation.Y {type=MovieSceneScriptingFloatChannel, default=0}
       head_ctrl.Rotation.Z {type=MovieSceneScriptingFloatChannel, default=0}
       head_ctrl.Scale.X {type=MovieSceneScriptingFloatChannel, default=1}
       head_ctrl.Scale.Y {type=MovieSceneScriptingFloatChannel, default=1}
       head_ctrl.Scale.Z {type=MovieSceneScriptingFloatChannel, default=1}
       clavicle_l_ctrl.Location.X {type=MovieSceneScriptingFloatChannel, default=0}
       clavicle_l_ctrl.Location.Y {type=MovieSceneScriptingFloatChannel, default=0}
       clavicle_l_ctrl.Location.Z {type=MovieSceneScriptingFloatChannel, default=0}
       clavicle_l_ctrl.Rotation.X {type=MovieSceneScriptingFloatChannel, default=0}
       clavicle_l_ctrl.Rotation.Y {type=MovieSceneScriptingFloatChannel, default=0}
       clavicle_l_ctrl.Rotation.Z {type=MovieSceneScriptingFloatChannel, default=0}
       clavicle_l_ctrl.Scale.X {type=MovieSceneScriptingFloatChannel, default=1}
       clavicle_l_ctrl.Scale.Y {type=MovieSceneScriptingFloatChannel, default=1}
       clavicle_l_ctrl.Scale.Z {type=MovieSceneScriptingFloatChannel, default=1}
       clavicle_r_ctrl.Location.X {type=MovieSceneScriptingFloatChannel, default=0}
       clavicle_r_ctrl.Location.Y {type=MovieSceneScriptingFloatChannel, default=0}
       clavicle_r_ctrl.Location.Z {type=MovieSceneScriptingFloatChannel, default=0}
       clavicle_r_ctrl.Rotation.X {type=MovieSceneScriptingFloatChannel, default=0}
       clavicle_r_ctrl.Rotation.Y {type=MovieSceneScriptingFloatChannel, default=0}
       clavicle_r_ctrl.Rotation.Z {type=MovieSceneScriptingFloatChannel, default=0}
       clavicle_r_ctrl.Scale.X {type=MovieSceneScriptingFloatChannel, default=1}
       clavicle_r_ctrl.Scale.Y {type=MovieSceneScriptingFloatChannel, default=1}
       clavicle_r_ctrl.Scale.Z {type=MovieSceneScriptingFloatChannel, default=1}
       neck_01_ctrl.Location.X {type=MovieSceneScriptingFloatChannel, default=0}
       neck_01_ctrl.Location.Y {type=MovieSceneScriptingFloatChannel, default=0}
       neck_01_ctrl.Location.Z {type=MovieSceneScriptingFloatChannel, default=0}
       neck_01_ctrl.Rotation.X {type=MovieSceneScriptingFloatChannel, default=0}
       neck_01_ctrl.Rotation.Y {type=MovieSceneScriptingFloatChannel, default=0}
       neck_01_ctrl.Rotation.Z {type=MovieSceneScriptingFloatChannel, default=0}
       neck_01_ctrl.Scale.X {type=MovieSceneScriptingFloatChannel, default=1}
       neck_01_ctrl.Scale.Y {type=MovieSceneScriptingFloatChannel, default=1}
       neck_01_ctrl.Scale.Z {type=MovieSceneScriptingFloatChannel, default=1}
       neck_02_ctrl.Location.X {type=MovieSceneScriptingFloatChannel, default=0}
       neck_02_ctrl.Location.Y {type=MovieSceneScriptingFloatChannel, default=0}
       neck_02_ctrl.Location.Z {type=MovieSceneScriptingFloatChannel, default=0}
       neck_02_ctrl.Rotation.X {type=MovieSceneScriptingFloatChannel, default=0}
       neck_02_ctrl.Rotation.Y {type=MovieSceneScriptingFloatChannel, default=0}
       neck_02_ctrl.Rotation.Z {type=MovieSceneScriptingFloatChannel, default=0}
       neck_02_ctrl.Scale.X {type=MovieSceneScriptingFloatChannel, default=1}
       neck_02_ctrl.Scale.Y {type=MovieSceneScriptingFloatChannel, default=1}
       neck_02_ctrl.Scale.Z {type=MovieSceneScriptingFloatChannel, default=1}
       head_ik_ctrl.Location.X {type=MovieSceneScriptingFloatChannel, default=0}
       head_ik_ctrl.Location.Y {type=MovieSceneScriptingFloatChannel, default=0}
       head_ik_ctrl.Location.Z {type=MovieSceneScriptingFloatChannel, default=0}
       head_ik_ctrl.Rotation.X {type=MovieSceneScriptingFloatChannel, default=0}
       head_ik_ctrl.Rotation.Y {type=MovieSceneScriptingFloatChannel, default=0}
       head_ik_ctrl.Rotation.Z {type=MovieSceneScriptingFloatChannel, default=0}
       head_ik_ctrl.Scale.X {type=MovieSceneScriptingFloatChannel, default=1}
       head_ik_ctrl.Scale.Y {type=MovieSceneScriptingFloatChannel, default=1}
       head_ik_ctrl.Scale.Z {type=MovieSceneScriptingFloatChannel, default=1}
       chest_ctrl.Location.X {type=MovieSceneScriptingFloatChannel, default=0}
       chest_ctrl.Location.Y {type=MovieSceneScriptingFloatChannel, default=0}
       chest_ctrl.Location.Z {type=MovieSceneScriptingFloatChannel, default=0}
       chest_ctrl.Rotation.X {type=MovieSceneScriptingFloatChannel, default=0}
       chest_ctrl.Rotation.Y {type=MovieSceneScriptingFloatChannel, default=0}
       chest_ctrl.Rotation.Z {type=MovieSceneScriptingFloatChannel, default=0}
       chest_ctrl.Scale.X {type=MovieSceneScriptingFloatChannel, default=1}
       chest_ctrl.Scale.Y {type=MovieSceneScriptingFloatChannel, default=1}
       chest_ctrl.Scale.Z {type=MovieSceneScriptingFloatChannel, default=1}
       chest_tan_ctrl.Location.X {type=MovieSceneScriptingFloatChannel, default=0}
       chest_tan_ctrl.Location.Y {type=MovieSceneScriptingFloatChannel, default=0}
       chest_tan_ctrl.Location.Z {type=MovieSceneScriptingFloatChannel, default=0}
       chest_tan_ctrl.Rotation.X {type=MovieSceneScriptingFloatChannel, default=0}
       chest_tan_ctrl.Rotation.Y {type=MovieSceneScriptingFloatChannel, default=0}
       chest_tan_ctrl.Rotation.Z {type=MovieSceneScriptingFloatChannel, default=0}
       chest_tan_ctrl.Scale.X {type=MovieSceneScriptingFloatChannel, default=1}
       chest_tan_ctrl.Scale.Y {type=MovieSceneScriptingFloatChannel, default=1}
       chest_tan_ctrl.Scale.Z {type=MovieSceneScriptingFloatChannel, default=1}
      ]
     }]
   }
  ]
 }]
display_rate {denominator=1, numerator=30}
tick_resolution {denominator=1, numerator=24000}
tracks []
